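# Control Rig for the SkeletonHunterGlitch male medium player body that procedurally poses coat bones from thigh movement.
# Spherical pose readers on both thighs, scaled by Alpha_sliding, weight transforms on left, right and back coat master bones.
# A third right-thigh reader, scaled by Alpha_jump_right, adjusts the front-right master.
# It also chains a full-weight, ungated offset on dyn_coat_r_02; Alpha_crouching is declared but unused.
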ControlRigBlueprint /Game/Characters/Player/Male/Medium/Bodies/M_MED_SkeletonHunterGlitch/Meshes/M_MED_SkeletonHunterGlitch_CtrlRig
graphs [
 RigVMFunctionLibrary {links [], nodes []}
 RigVMModel {nodes [
   RigUnit_BeginExecution = &RigUnit_BeginExecution()
   SequenceAggregate = &RigVMFunction_Sequence(ExecuteContext=@RigUnit_BeginExecution.ExecutePin)
   SequenceAggregate_1 = &RigVMFunction_Sequence(ExecuteContext=@SequenceAggregate.A)
   n1 = &RigUnit_SphericalPoseReader(
    ActiveRegionScaleFactors=(PositiveWidth=3.5,NegativeWidth=1,PositiveHeight=3,NegativeHeight=3)
    ActiveRegionSize=0.1
    Debug=(bDrawDebug=false,bDraw2D=false,DebugScale=25,DebugSegments=20,DebugThickness=0.25)
    DriverAxis=(X=-1,Y=0,Z=0)
    DriverItem=(Type=Bone,Name="thigh_l")
    FalloffRegionScaleFactors=(PositiveWidth=1,NegativeWidth=1,PositiveHeight=1,NegativeHeight=1)
    FalloffSize=1
    OptionalParentItem=(Type=Bone)
    RotationOffset=(X=90,Y=0,Z=0)
    ExecutePin=@SequenceAggregate_1.A
   )
   n2 = &RigVMFunction_MathDoubleMul(A=@n1.OutputParam, B=RigVMVariableNode() {variable=Alpha_sliding}.Value)
   > &RigUnit_ModifyTransforms(
    ItemToModify=((Item=(Type=Bone,Name="dyn_coat_fr_l_MSTR"),Transform=(Rotation=(X=0,Y=0,Z=0,W=1),Translation=(X=5,Y=0,Z=0),Scale3D=(X=1,Y=1,Z=1))))
    ItemToModify.0.Transform.Rotation=&RigVMFunction_MathQuaternionFromEuler(Euler=(X=0,Y=-60,Z=0)).Result
    Weight=@n2.Result
   )
   > &RigUnit_ModifyTransforms(
    ItemToModify=((Item=(Type=Bone,Name="dyn_coat_l_MSTR"),Transform=(Rotation=(X=0,Y=0,Z=0,W=1),Translation=(X=0,Y=0,Z=0),Scale3D=(X=1,Y=1,Z=1))))
    ItemToModify.0.Transform.Rotation=&RigVMFunction_MathQuaternionFromEuler(Euler=(X=30,Y=-120,Z=0)).Result
    Weight=@n2.Result
   )
   > &RigUnit_ModifyTransforms(
    ItemToModify=((Item=(Type=Bone,Name="dyn_coat_bk_l_MSTR"),Transform=(Rotation=(X=0,Y=0,Z=0,W=1),Translation=(X=0,Y=0,Z=0),Scale3D=(X=1,Y=1,Z=1))))
    ItemToModify.0.Transform.Rotation=&RigVMFunction_MathQuaternionFromEuler(Euler=(X=0,Y=-120,Z=0)).Result
    Weight=@n2.Result
   )
   > &RigUnit_ModifyTransforms(
    ItemToModify=((Item=(Type=Bone,Name="dyn_coat_bk_MSTR"),Transform=(Rotation=(X=0,Y=0,Z=0,W=1),Translation=(X=0,Y=0,Z=0),Scale3D=(X=1,Y=1,Z=1))))
    ItemToModify.0.Transform.Rotation=&RigVMFunction_MathQuaternionFromEuler(Euler=(X=0,Y=-120,Z=0)).Result
    Weight=@n2.Result
   )
   n3 = &RigUnit_SphericalPoseReader(
    ActiveRegionScaleFactors=(PositiveWidth=1,NegativeWidth=1,PositiveHeight=3,NegativeHeight=3)
    ActiveRegionSize=0.1
    Debug=(bDrawDebug=false,DebugScale=25,DebugSegments=20,DebugThickness=0.25)
    DriverAxis=(X=0.25,Y=0,Z=0)
    DriverItem=(Type=Bone,Name="thigh_r")
    FalloffRegionScaleFactors=(PositiveWidth=1,NegativeWidth=1,PositiveHeight=1,NegativeHeight=1)
    FalloffSize=1
    FlipWidthScaling=false
    OptionalParentItem=(Type=Bone)
    RotationOffset=(X=-90,Y=0,Z=0)
    ExecutePin=@SequenceAggregate_1.B
   )
   n4 = &RigVMFunction_MathDoubleMul(A=@n3.OutputParam, B=RigVMVariableNode() {variable=Alpha_sliding}.Value)
   > &RigUnit_ModifyTransforms(
    ItemToModify=((Item=(Type=Bone,Name="dyn_coat_fr_r_MSTR"),Transform=(Rotation=(X=0,Y=0,Z=0,W=1),Translation=(X=0,Y=0,Z=0),Scale3D=(X=1,Y=1,Z=1))))
    ItemToModify.0.Transform.Rotation=&RigVMFunction_MathQuaternionFromEuler(Euler=(X=0,Y=-120,Z=0)).Result
    Weight=@n4.Result
   )
   > &RigUnit_ModifyTransforms(
    ItemToModify=((Item=(Type=Bone,Name="dyn_coat_r_MSTR"),Transform=(Rotation=(X=0,Y=0,Z=0,W=1),Translation=(X=0,Y=0,Z=0),Scale3D=(X=1,Y=1,Z=1))))
    ItemToModify.0.Transform.Rotation=&RigVMFunction_MathQuaternionFromEuler(Euler=(X=0,Y=-120,Z=0)).Result
    Weight=@n4.Result
   )
   > &RigUnit_ModifyTransforms(
    ItemToModify=((Item=(Type=Bone,Name="dyn_coat_bk_r_MSTR"),Transform=(Rotation=(X=0,Y=0,Z=0,W=1),Translation=(X=0,Y=0,Z=0),Scale3D=(X=1,Y=1,Z=1))))
    ItemToModify.0.Transform.Rotation=&RigVMFunction_MathQuaternionFromEuler(Euler=(X=0,Y=-120,Z=0)).Result
    Weight=@n4.Result
   )
   n5 = &RigUnit_SphericalPoseReader(
    ActiveRegionScaleFactors=(PositiveWidth=2,NegativeWidth=2,PositiveHeight=2,NegativeHeight=2)
    ActiveRegionSize=0.07
    Debug=(bDrawDebug=false,bDraw2D=false,bDrawLocalAxes=false,DebugScale=25,DebugSegments=20,DebugThickness=0.25)
    DriverAxis=(X=1,Y=0,Z=0)
    DriverItem=(Type=Bone,Name="thigh_r")
    FalloffRegionScaleFactors=(PositiveWidth=1,NegativeWidth=1,PositiveHeight=1,NegativeHeight=1)
    FalloffSize=0.3
    OptionalParentItem=(Type=Bone,Name="None")
    RotationOffset=(X=290,Y=20,Z=15)
    ExecutePin=@SequenceAggregate.B
   )
   > &RigUnit_ModifyTransforms(
    ItemToModify=((Item=(Type=Bone,Name="dyn_coat_fr_r_MSTR"),Transform=(Rotation=(X=0,Y=0,Z=0,W=1),Translation=(X=0,Y=8,Z=4),Scale3D=(X=1,Y=1,Z=1))))
    ItemToModify.0.Transform.Rotation=&RigVMFunction_MathQuaternionFromEuler(Euler=(X=-20,Y=5,Z=0)).Result
    Weight=&RigVMFunction_MathDoubleMul(A=@n5.OutputParam, B=RigVMVariableNode() {variable=Alpha_jump_right}.Value).Result
   )
   > &RigUnit_ModifyTransforms(
    ItemToModify=((Item=(Type=Bone,Name="dyn_coat_r_02"),Transform=(Rotation=(X=0,Y=0,Z=0,W=1),Translation=(X=0,Y=-5,Z=0),Scale3D=(X=1,Y=1,Z=1))))
    Weight=1
    ItemToModify.0.Transform.Rotation=&RigVMFunction_MathQuaternionFromEuler(Euler=(X=0,Y=5,Z=0)).Result
   )
  ]}
]
hierarchy [
 root {loc=<0 0 0>, rot=<0 0 0 1>}
 dyn_root_simspace {parent=root, loc=<0 0 0>, rot=<0 0 0 1>}
 floorCOL {parent=root, loc=<0 0 0>, rot=<0 0 0 1>}
 pelvis {parent=root, loc=<0 -0.7071 93.1191>, rot=<0.0127 0.707 -0.0127 -0.707>}
 spine_01 {parent=pelvis, loc=<6.7494 -0.6388 0>, rot=<0 0 0.034 -0.9994>}
 spine_02 {parent=spine_01, loc=<7.2518 -0.035 0>, rot=<0 0 0.0238 -0.9997>}
 spine_03 {parent=spine_02, loc=<6.0104 0 0>, rot=<0 0 0.0149 -0.9999>}
 spine_04 {parent=spine_03, loc=<8.1024 0 0>, rot=<0 0 -0.0132 -0.9999>}
 spine_05 {parent=spine_04, loc=<10.0146 0 0>, rot=<0 0 -0.0015 -1>}
 attach_cape {parent=spine_05, loc=<8.909 -10.365 -0.1088>, rot=<0.0284 -0.7065 -0.0284 -0.7065>}
 neck_01 {parent=spine_05, loc=<11.3138 0.1529 0>, rot=<0 0 -0.2003 -0.9797>}
 neck_02 {parent=neck_01, loc=<4.5627 0 0>, rot=<0 0 0.0504 -0.9987>}
 head {parent=neck_02, loc=<5.3914 0 0>, rot=<0 0 0.115 -0.9934>}
 attach_backpack {parent=spine_05, loc=<-0.123 -11.728 0>, rot=<0 -0.7071 0 -0.7071>}
 clavicle_l {parent=spine_05, loc=<4.3575 -0.1466 -2.3223>, rot=<0.0465 -0.7289 -0.009 -0.6829>}
 upperarm_l {parent=clavicle_l, loc=<15.5111 0 0>, rot=<0.0226 -0.3234 -0.0088 -0.9459>}
 lowerarm_l {parent=upperarm_l, loc=<26.6514 0 0>, rot=<0.0391 -0.0099 -0.2453 -0.9686>}
 hand_l {parent=lowerarm_l, loc=<22.3491 0 0>, rot=<-0.7081 -0.0507 0.0975 -0.6975>}
 thumb_01_l {parent=hand_l, loc=<1.3118 -0.3191 -2.459>, rot=<0.6546 -0.4042 0.0664 -0.6354>}
 thumb_02_l {parent=thumb_01_l, loc=<5.9336 -0.0448 0.0401>, rot=<-0.0037 0.0213 0.298 -0.9543>}
 thumb_03_l {parent=thumb_02_l, loc=<2.2223 0.3682 -0.0907>, rot=<0.0118 -0.013 0.0066 -0.9998>}
 index_metacarpal_l {parent=hand_l, loc=<1.6535 0.0183 -2.96>, rot=<0.0158 -0.0124 -0.0173 -0.9996>}
 index_01_l {parent=index_metacarpal_l, loc=<8.3008 0.0357 -0.0524>, rot=<-0.0044 0.0485 0.0924 -0.9945>}
 index_02_l {parent=index_01_l, loc=<3.3587 -0.3004 0.0452>, rot=<-0.002 0.0027 0.3225 -0.9466>}
 index_03_l {parent=index_02_l, loc=<2.5578 -0.123 -0.0252>, rot=<-0.002 0.0067 0.2558 -0.9667>}
 middle_metacarpal_l {parent=hand_l, loc=<2.1856 -0.1725 -0.5984>, rot=<0.006 0.0168 -0.0329 -0.9993>}
 middle_01_l {parent=middle_metacarpal_l, loc=<7.3275 -0.2447 0.0129>, rot=<-0.0208 0.0042 0.0858 -0.9961>}
 middle_02_l {parent=middle_01_l, loc=<4.0123 0.1401 0.1175>, rot=<-0.0218 0.0258 0.3512 -0.9357>}
 middle_03_l {parent=middle_02_l, loc=<2.6077 0.0534 0.1373>, rot=<0.0082 0.0188 0.1708 -0.9851>}
 ring_metacarpal_l {parent=hand_l, loc=<2.2074 -0.3625 1.3638>, rot=<-0.0284 0.0722 -0.0458 -0.9959>}
 ring_01_l {parent=ring_metacarpal_l, loc=<7.0295 -0.6197 -0.1446>, rot=<-0.0489 -0.0252 0.1892 -0.9804>}
 ring_02_l {parent=ring_01_l, loc=<3.6174 0.4647 -0.1027>, rot=<0 0 0.2806 -0.9598>}
 ring_03_l {parent=ring_02_l, loc=<2.4152 0.114 0.0199>, rot=<0 0 0.1189 -0.9929>}
 pinky_metacarpal_l {parent=hand_l, loc=<1.7616 -0.9753 3.3814>, rot=<-0.0397 0.0814 -0.069 -0.9935>}
 pinky_01_l {parent=pinky_metacarpal_l, loc=<7.1709 -0.4831 -0.1094>, rot=<-0.0654 -0.0035 0.2242 -0.9723>}
 pinky_02_l {parent=pinky_01_l, loc=<2.8274 0.0893 -0.1043>, rot=<0.0037 0.0011 0.2854 -0.9584>}
 pinky_03_l {parent=pinky_02_l, loc=<2.2046 0.0307 -0.0078>, rot=<-0.0002 0.0098 0.1445 -0.9895>}
 deform_wrist_dn_base_l {parent=hand_l, loc=<0 -2.2342 0>, rot=<0 0 0 -1>}
 deform_wrist_dn_l {parent=deform_wrist_dn_base_l, loc=<0 0 0>, rot=<0 0 0 1>}
 deform_wrist_up_base_l {parent=hand_l, loc=<0 2.2682 0>, rot=<0 0 0 -1>}
 deform_wrist_up_l {parent=deform_wrist_up_base_l, loc=<0 0 0>, rot=<0 0 0 1>}
 lowerarm_twist_02_l {parent=lowerarm_l, loc=<9.6199 0 0>, rot=<0 0 0 -1>}
 lowerarm_twist_01_l {parent=lowerarm_l, loc=<17.8549 0 0>, rot=<0 0 0 -1>}
 deform_elbow_in_base_l {parent=lowerarm_l, loc=<0.2112 5.259 0.8274>, rot=<0 0 0 -1>}
 deform_elbow_in_l {parent=deform_elbow_in_base_l, loc=<0 0 0>, rot=<0 0 0 1>}
 deform_elbow_out_base_l {parent=lowerarm_l, loc=<0 -4 0>, rot=<0 0 0 -1>}
 deform_elbow_out_l {parent=deform_elbow_out_base_l, loc=<0 0 0>, rot=<0 0 0 1>}
 upperarm_twist_01_l {parent=upperarm_l, loc=<0.0041 0 0>, rot=<0 0 0 1>}
 deform_upperarm_bk_base_l {parent=upperarm_twist_01_l, loc=<1.0813 -7.2193 -1.9453>, rot=<0 0 0 -1>}
 deform_upperarm_bk_l {parent=deform_upperarm_bk_base_l, loc=<0 0 0>, rot=<0 0 0 1>}
 deform_upperarm_fr_base_l {parent=upperarm_twist_01_l, loc=<-0.3691 7.0506 -2.765>, rot=<0 0 0 -1>}
 deform_upperarm_fr_l {parent=deform_upperarm_fr_base_l, loc=<0 0 0>, rot=<0 0 0 1>}
 deform_upperarm_up_base_l {parent=upperarm_twist_01_l, loc=<-3.2216 -1.1391 3.3311>, rot=<0 0 0 -1>}
 deform_upperarm_up_l {parent=deform_upperarm_up_base_l, loc=<0 0 0>, rot=<0 0 0 1>}
 upperarm_twist_02_l {parent=upperarm_l, loc=<13.7941 0 0>, rot=<0 0 0 1>}
 clavicle_r {parent=spine_05, loc=<4.3575 -0.1466 2.3223>, rot=<0.7289 0.0465 -0.6829 0.009>}
 upperarm_r {parent=clavicle_r, loc=<-15.5111 0 0>, rot=<0.0226 -0.3234 -0.0088 -0.9459>}
 lowerarm_r {parent=upperarm_r, loc=<-26.6514 0 0>, rot=<0.0391 -0.0099 -0.2453 -0.9686>}
 hand_r {parent=lowerarm_r, loc=<-22.3491 0 0>, rot=<-0.7081 -0.0507 0.0975 -0.6975>}
 thumb_01_r {parent=hand_r, loc=<-1.3119 0.3192 2.459>, rot=<0.6546 -0.4042 0.0664 -0.6354>}
 thumb_02_r {parent=thumb_01_r, loc=<-5.9336 0.0448 -0.0401>, rot=<-0.0037 0.0213 0.298 -0.9543>}
 thumb_03_r {parent=thumb_02_r, loc=<-2.2223 -0.3682 0.0907>, rot=<0.0118 -0.013 0.0066 -0.9998>}
 index_metacarpal_r {parent=hand_r, loc=<-1.6534 -0.0181 2.9601>, rot=<0.0158 -0.0124 -0.0173 -0.9996>}
 index_01_r {parent=index_metacarpal_r, loc=<-8.3008 -0.0357 0.0524>, rot=<-0.0044 0.0485 0.0924 -0.9945>}
 index_02_r {parent=index_01_r, loc=<-3.3589 0.3003 -0.0453>, rot=<-0.002 0.0027 0.3225 -0.9466>}
 index_03_r {parent=index_02_r, loc=<-2.5578 0.123 0.0252>, rot=<-0.002 0.0067 0.2558 -0.9667>}
 middle_metacarpal_r {parent=hand_r, loc=<-2.1856 0.1726 0.5984>, rot=<0.006 0.0168 -0.0329 -0.9993>}
 middle_01_r {parent=middle_metacarpal_r, loc=<-7.3273 0.2447 -0.0129>, rot=<-0.0208 0.0042 0.0858 -0.9961>}
 middle_02_r {parent=middle_01_r, loc=<-4.0142 -0.1411 -0.07>, rot=<-0.0217 0.0259 0.3512 -0.9357>}
 middle_03_r {parent=middle_02_r, loc=<-2.6074 -0.0534 -0.1312>, rot=<0.0083 0.0188 0.1708 -0.9851>}
 ring_metacarpal_r {parent=hand_r, loc=<-2.2073 0.3627 -1.3638>, rot=<-0.0284 0.0722 -0.0458 -0.9959>}
 ring_01_r {parent=ring_metacarpal_r, loc=<-7.0297 0.6197 0.1446>, rot=<-0.0489 -0.0252 0.1892 -0.9804>}
 ring_02_r {parent=ring_01_r, loc=<-3.6171 -0.4648 0.1027>, rot=<0 0 0.2806 -0.9598>}
 ring_03_r {parent=ring_02_r, loc=<-2.4152 -0.114 -0.0199>, rot=<0 0 0.1189 -0.9929>}
 pinky_metacarpal_r {parent=hand_r, loc=<-1.7617 0.9753 -3.3814>, rot=<-0.0397 0.0814 -0.069 -0.9935>}
 pinky_01_r {parent=pinky_metacarpal_r, loc=<-7.1707 0.4832 0.1094>, rot=<-0.0654 -0.0035 0.2242 -0.9723>}
 pinky_02_r {parent=pinky_01_r, loc=<-2.8274 -0.0893 0.1043>, rot=<0.0037 0.0011 0.2854 -0.9584>}
 pinky_03_r {parent=pinky_02_r, loc=<-2.2045 -0.031 0.0078>, rot=<-0.0002 0.0098 0.1445 -0.9895>}
 deform_wrist_dn_base_r {parent=hand_r, loc=<-0.0004 2.2345 -0.0001>, rot=<0 0 0 -1>}
 deform_wrist_dn_r {parent=deform_wrist_dn_base_r, loc=<0 0 0>, rot=<0 0 0 1>}
 deform_wrist_up_base_r {parent=hand_r, loc=<0.0003 -2.2684 0.0001>, rot=<0 0 0 -1>}
 deform_wrist_up_r {parent=deform_wrist_up_base_r, loc=<0 0 0>, rot=<0 0 0 1>}
 lowerarm_twist_02_r {parent=lowerarm_r, loc=<-9.6199 0 0>, rot=<0 0 0 -1>}
 lowerarm_twist_01_r {parent=lowerarm_r, loc=<-17.8549 0 0>, rot=<0 0 0 -1>}
 deform_elbow_in_base_r {parent=lowerarm_r, loc=<-0.2111 -5.2591 -0.8275>, rot=<0 0 0 -1>}
 deform_elbow_in_r {parent=deform_elbow_in_base_r, loc=<-0.0001 0 0.0002>, rot=<0 0 0 1>}
 deform_elbow_out_base_r {parent=lowerarm_r, loc=<0 4 -0.0001>, rot=<0 0 0 -1>}
 deform_elbow_out_r {parent=deform_elbow_out_base_r, loc=<0 0 0>, rot=<0 0 0 1>}
 upperarm_twist_01_r {parent=upperarm_r, loc=<-0.0041 0 0>, rot=<0 0 0 -1>}
 deform_upperarm_bk_base_r {parent=upperarm_twist_01_r, loc=<-1.0814 7.2192 1.9456>, rot=<0 0 0 -1>}
 deform_upperarm_bk_r {parent=deform_upperarm_bk_base_r, loc=<0 0 0>, rot=<0 0 0 1>}
 deform_upperarm_fr_base_r {parent=upperarm_twist_01_r, loc=<0.3689 -7.0506 2.7652>, rot=<0 0 0 -1>}
 deform_upperarm_fr_r {parent=deform_upperarm_fr_base_r, loc=<0 0 0>, rot=<0 0 0 1>}
 deform_upperarm_up_base_r {parent=upperarm_twist_01_r, loc=<3.2214 1.1391 -3.3308>, rot=<0 0 0 -1>}
 deform_upperarm_up_r {parent=deform_upperarm_up_base_r, loc=<0 0 0>, rot=<0 0 0 -1>}
 upperarm_twist_02_r {parent=upperarm_r, loc=<-13.7941 0 0>, rot=<0 0 0 -1>}
 extra_jaket_collar_l {parent=spine_05, loc=<-4.122 12.839 -7.6628>, rot=<0.1208 -0.3703 0.1921 -0.9008>}
 extra_jaket_collar_r {parent=spine_05, loc=<-4.1221 12.839 7.6628>, rot=<-0.3703 -0.1208 0.9008 0.1921>}
 deform_pec_aim_base_l {parent=spine_04, loc=<7.8732 12.6388 -2.4123>, rot=<0.0102 -0.4602 0.3979 -0.7936>}
 deform_pec_aim_l {parent=deform_pec_aim_base_l, loc=<0 0 0>, rot=<0 0 0 1>}
 deform_pec_base_l {parent=deform_pec_aim_l, loc=<10.766 0 0>, rot=<0 0 0 1>}
 deform_pec_l {parent=deform_pec_base_l, loc=<0.055 0 0>, rot=<0 0 0 1>}
 deform_pec_aim_base_r {parent=spine_04, loc=<7.8727 12.6388 2.4123>, rot=<-0.4602 -0.0102 0.7936 0.3979>}
 deform_pec_aim_r {parent=deform_pec_aim_base_r, loc=<0 0 0>, rot=<0 0 0 1>}
 deform_pec_base_r {parent=deform_pec_aim_r, loc=<-10.7662 -0.0006 -0.0007>, rot=<0 0 0 1>}
 deform_pec_r {parent=deform_pec_base_r, loc=<0 0 0>, rot=<0 0 0 1>}
 thigh_l {parent=pelvis, loc=<-2.5693 -0.0109 -7.7501>, rot=<0.0012 0.0572 0.0155 -0.9982>}
 calf_l {parent=thigh_l, loc=<-39.0484 0 0>, rot=<0.0006 -0.0161 -0.0391 -0.9991>}
 foot_l {parent=calf_l, loc=<-42.561 0.1077 0.4292>, rot=<0.0418 -0.0455 0.0343 -0.9975>}
 ball_l {parent=foot_l, loc=<-6.1599 13.8264 0.6806>, rot=<0.0001 -0.0001 -0.7063 -0.7079>}
 calf_twist_02_l {parent=calf_l, loc=<-18.8064 -0.0967 0.0397>, rot=<0.0237 0.0095 -0.0035 -0.9997>}
 calf_twist_01_l {parent=calf_l, loc=<-35.813 -0.0967 0.0397>, rot=<0.0237 0.0095 -0.0035 -0.9997>}
 deform_knee_in_base_l {parent=calf_l, loc=<1.298 -5.2957 -0.0496>, rot=<0 0 0 -1>}
 deform_knee_in_l {parent=deform_knee_in_base_l, loc=<0 0 0>, rot=<0 0 0 1>}
 deform_knee_out_base_l {parent=calf_l, loc=<0 4 0>, rot=<0 0 0 -1>}
 deform_knee_out_l {parent=deform_knee_out_base_l, loc=<0 0 0>, rot=<0 0 0 1>}
 thigh_twist_01_l {parent=thigh_l, loc=<-21.6303 0 0>, rot=<0.0013 0 0 -1>}
 deform_groin_base_l {parent=thigh_l, loc=<-2.1908 8.1142 -1.7442>, rot=<0 0 0 1>}
 deform_groin_l {parent=deform_groin_base_l, loc=<0 0 0>, rot=<0 0 0 1>}
 thigh_r {parent=pelvis, loc=<-2.5691 -0.0109 7.7501>, rot=<0.0572 -0.0012 0.9982 0.0155>}
 calf_r {parent=thigh_r, loc=<39.0487 0 0>, rot=<0.0006 -0.0161 -0.0391 -0.9991>}
 foot_r {parent=calf_r, loc=<42.5612 -0.1077 -0.4292>, rot=<0.0418 -0.0455 0.0343 -0.9975>}
 ball_r {parent=foot_r, loc=<6.1599 -13.8264 -0.6806>, rot=<0.0001 -0.0001 -0.7063 -0.7079>}
 calf_twist_02_r {parent=calf_r, loc=<18.8064 0.0967 -0.0397>, rot=<0.0237 0.0095 -0.0035 -0.9997>}
 calf_twist_01_r {parent=calf_r, loc=<35.8132 0.0967 -0.0397>, rot=<0.0237 0.0095 -0.0035 -0.9997>}
 deform_knee_in_base_r {parent=calf_r, loc=<-1.298 5.2957 0.0496>, rot=<0 0 0 -1>}
 deform_knee_in_r {parent=deform_knee_in_base_r, loc=<0 0 0>, rot=<0 0 0 1>}
 deform_knee_out_base_r {parent=calf_r, loc=<0 -4 0>, rot=<0 0 0 -1>}
 deform_knee_out_r {parent=deform_knee_out_base_r, loc=<0 0 0>, rot=<0 0 0 1>}
 thigh_twist_01_r {parent=thigh_r, loc=<21.6303 0 0>, rot=<0.0013 0 0 -1>}
 deform_groin_base_r {parent=thigh_r, loc=<2.1908 -8.1142 1.7442>, rot=<0 0 0 -1>}
 deform_groin_r {parent=deform_groin_base_r, loc=<0 0 0>, rot=<0 0 0 1>}
 dyn_jacketFront_mstr_l {parent=pelvis, loc=<11.2633 5.2396 -12.535>, rot=<-0.0354 -0.9917 -0.0491 -0.1138>}
 dyn_jacketFront_01_l {parent=dyn_jacketFront_mstr_l, loc=<0 0 0>, rot=<0 0.0614 0 -0.9981>}
 dyn_jacketFront_02_l {parent=dyn_jacketFront_01_l, loc=<17 0 0>, rot=<0.0172 -0.0001 0.0157 -0.9997>}
 dyn_jacketFront_03_l {parent=dyn_jacketFront_02_l, loc=<17 0 0>, rot=<-0.0062 0.0009 0.0018 -1>}
 dyn_jacketFront_04_l {parent=dyn_jacketFront_03_l, loc=<17 0 0>, rot=<0.0093 -0.0031 -0.0033 -0.9999>}
 dyn_jacketFront_05_l {parent=dyn_jacketFront_04_l, loc=<17 0 0>, rot=<0 0 0 1>}
 dyn_jacketSide_mstr_l {parent=pelvis, loc=<11.128 -4.2926 -12.8191>, rot=<-0.0046 -0.9934 -0.0462 -0.105>}
 dyn_jacketSide_01_l {parent=dyn_jacketSide_mstr_l, loc=<0 0 0>, rot=<-0.3228 0.0485 0.0278 -0.9448>}
 dyn_jacketSide_02_l {parent=dyn_jacketSide_01_l, loc=<17 0 0>, rot=<0.0172 -0.0001 0.0157 -0.9997>}
 dyn_jacketSide_03_l {parent=dyn_jacketSide_02_l, loc=<17 0 0>, rot=<-0.0062 0.0009 0.0018 -1>}
 dyn_jacketSide_04_l {parent=dyn_jacketSide_03_l, loc=<17 0 0>, rot=<0.0093 -0.0031 -0.0033 -0.9999>}
 dyn_jacketSide_05_l {parent=dyn_jacketSide_04_l, loc=<17 0 0>, rot=<0 0 0 1>}
 dyn_jacketBack_mstr_l {parent=pelvis, loc=<11.1923 -9.509 -5.6499>, rot=<0.0457 -0.7922 0.6004 -0.0989>}
 dyn_jacketBack_01_l {parent=dyn_jacketBack_mstr_l, loc=<0 0 0>, rot=<0 0 0 1>}
 dyn_jacketBack_02_l {parent=dyn_jacketBack_01_l, loc=<17 0 0>, rot=<0.0271 0.0057 0.0002 -0.9996>}
 dyn_jacketBack_03_l {parent=dyn_jacketBack_02_l, loc=<17 0 0>, rot=<-0.0184 0.0097 -0.0002 -0.9998>}
 dyn_jacketBack_04_l {parent=dyn_jacketBack_03_l, loc=<17 0 0>, rot=<-0.0016 -0.0003 0 -1>}
 dyn_jacketBack_05_l {parent=dyn_jacketBack_04_l, loc=<17 0 0>, rot=<0 0 0 1>}
 dyn_jacketBack_mstr_r {parent=pelvis, loc=<11.1924 -9.509 5.6499>, rot=<0.0832 -0.6012 0.794 -0.0338>}
 dyn_jacketBack_01_r {parent=dyn_jacketBack_mstr_r, loc=<0 0 0>, rot=<0 0 0 1>}
 dyn_jacketBack_02_r {parent=dyn_jacketBack_01_r, loc=<16.9868 0 0.6734>, rot=<-0.0271 0.0255 -0.0007 -0.9993>}
 dyn_jacketBack_03_r {parent=dyn_jacketBack_02_r, loc=<17 0 0.0001>, rot=<-0.0406 0.0097 -0.0004 -0.9991>}
 dyn_jacketBack_04_r {parent=dyn_jacketBack_03_r, loc=<17 0 0>, rot=<-0.0001 0 -0.0003 -1>}
 dyn_jacketBack_05_r {parent=dyn_jacketBack_04_r, loc=<17 0 0>, rot=<0 0 0 1>}
 dyn_jacketSide_mstr_r {parent=pelvis, loc=<11.1282 -4.2927 12.8191>, rot=<0.1472 -0.2801 0.9485 0.0128>}
 dyn_jacketSide_01_r {parent=dyn_jacketSide_mstr_r, loc=<0 0 0>, rot=<0 0 0 1>}
 dyn_jacketSide_02_r {parent=dyn_jacketSide_01_r, loc=<17.0002 0 0>, rot=<0 0 0 1>}
 dyn_jacketSide_03_r {parent=dyn_jacketSide_02_r, loc=<16.9916 0.5353 0.0059>, rot=<0 0 0 1>}
 dyn_jacketSide_04_r {parent=dyn_jacketSide_03_r, loc=<16.9895 0.5957 0.0375>, rot=<0 0 0 1>}
 dyn_jacketSide_05_r {parent=dyn_jacketSide_04_r, loc=<16.9929 0.4867 -0.071>, rot=<0 0 0 1>}
 dyn_jacketFront_mstr_r {parent=pelvis, loc=<11.2631 5.2393 12.5352>, rot=<0.1136 0.0332 0.9924 0.0337>}
 dyn_jacketFront_01_r {parent=dyn_jacketFront_mstr_r, loc=<0 0 0>, rot=<0 0.0614 0 -0.9981>}
 dyn_jacketFront_02_r {parent=dyn_jacketFront_01_r, loc=<17 0 0>, rot=<0.0544 0.002 -0.0124 -0.9984>}
 dyn_jacketFront_03_r {parent=dyn_jacketFront_02_r, loc=<17 0 0>, rot=<0.0063 0.0012 -0.0016 -1>}
 dyn_jacketFront_04_r {parent=dyn_jacketFront_03_r, loc=<17 0 0>, rot=<-0.0093 -0.0036 0.0027 -0.9999>}
 dyn_jacketFront_05_r {parent=dyn_jacketFront_04_r, loc=<17 0 0>, rot=<0 0 0 1>}
 deform_glute_base_l {parent=pelvis, loc=<-8.8064 -8.8222 -8.4287>, rot=<0.0275 0.0501 -0.4472 -0.8926>}
 deform_glute_l {parent=deform_glute_base_l, loc=<0 0 0>, rot=<0 0 0 1>}
 deform_glute_base_r {parent=pelvis, loc=<-8.8062 -8.8222 8.4287>, rot=<0.0501 -0.0275 0.8926 -0.4472>}
 deform_glute_r {parent=deform_glute_base_r, loc=<0 0 0>, rot=<0 0 0 1>}
 dyn_coat_l_MSTR {parent=pelvis, loc=<13.5048 -2.0856 -12.4053>, rot=<-0.0002 -0.985 0.0886 -0.1479>}
 dyn_coat_l_01 {parent=dyn_coat_l_MSTR, loc=<0.3406 0 0>, rot=<0 0 0 1>}
 dyn_coat_l_02 {parent=dyn_coat_l_01, loc=<8.4103 0 0>, rot=<0 0 0 1>}
 dyn_coat_l_03 {parent=dyn_coat_l_02, loc=<7.8528 0 0>, rot=<0 0 0 1>}
 dyn_coat_l_04 {parent=dyn_coat_l_03, loc=<16.9365 0 0>, rot=<0 0 0 1>}
 dyn_coat_l_05 {parent=dyn_coat_l_04, loc=<17.3764 0 0>, rot=<0 0 0 1>}
 dyn_coat_l_06 {parent=dyn_coat_l_05, loc=<19.0063 0 0>, rot=<0 0 0 1>}
 dyn_coat_fr_l_MSTR {parent=pelvis, loc=<16.7174 6.8826 -10.1505>, rot=<0.0178 -0.9842 -0.1444 -0.1008>}
 dyn_coat_fr_l_01 {parent=dyn_coat_fr_l_MSTR, loc=<0.3406 0 0>, rot=<0 0 0 1>}
 dyn_coat_fr_l_02 {parent=dyn_coat_fr_l_01, loc=<10.8407 0 0.6162>, rot=<0 0.0824 0 -0.9966>}
 dyn_coat_fr_l_03 {parent=dyn_coat_fr_l_02, loc=<7.9522 0 0>, rot=<0 0 0 1>}
 dyn_coat_fr_l_04 {parent=dyn_coat_fr_l_03, loc=<16.2735 -0.0058 0.003>, rot=<0 0 0 1>}
 dyn_coat_fr_l_05 {parent=dyn_coat_fr_l_04, loc=<16.9531 0.0058 -0.003>, rot=<0 0 0 1>}
 dyn_coat_fr_l_06 {parent=dyn_coat_fr_l_05, loc=<18.3953 0 0>, rot=<0 0 0 1>}
 dyn_coat_bk_MSTR {parent=pelvis, loc=<13.845 -9.9523 0.0405>, rot=<0.0654 -0.698 0.7101 -0.0662>}
 dyn_coat_bk_01 {parent=dyn_coat_bk_MSTR, loc=<0.3406 0 0>, rot=<0 0 0 1>}
 dyn_coat_bk_02 {parent=dyn_coat_bk_01, loc=<12.3256 0 0>, rot=<0 0 0 1>}
 dyn_coat_bk_03 {parent=dyn_coat_bk_02, loc=<13.4831 0 0>, rot=<0 0.0181 0 -0.9998>}
 dyn_coat_bk_04 {parent=dyn_coat_bk_03, loc=<13.9568 0 0>, rot=<0 0 0 1>}
 dyn_coat_bkl_05 {parent=dyn_coat_bk_04, loc=<2.3884 3.0189 0>, rot=<-0.0003 0.0103 -0.0283 -0.9995>}
 dyn_coat_bkl_06 {parent=dyn_coat_bkl_05, loc=<16.3107 0 0>, rot=<0 0 0 1>}
 dyn_coat_bkr_05 {parent=dyn_coat_bk_04, loc=<2.3948 -3.0121 0.1034>, rot=<0.0174 0.0108 0.0271 -0.9994>}
 dyn_coat_bkr_06 {parent=dyn_coat_bkr_05, loc=<16.3106 0 0>, rot=<0 0 0 1>}
 dyn_coat_bk_l_MSTR {parent=pelvis, loc=<11.7649 -7.579 -8.2398>, rot=<0.0341 -0.8411 0.5287 -0.1086>}
 dyn_coat_bk_l_01 {parent=dyn_coat_bk_l_MSTR, loc=<0.3406 0 0>, rot=<0 0 0 1>}
 dyn_coat_bk_l_02 {parent=dyn_coat_bk_l_01, loc=<8.8279 0 0>, rot=<0 0 0 1>}
 dyn_coat_bk_l_03 {parent=dyn_coat_bk_l_02, loc=<16.2192 0 0>, rot=<0 0.0178 0 -0.9998>}
 dyn_coat_bk_l_04 {parent=dyn_coat_bk_l_03, loc=<13.6501 0 0>, rot=<0 0.0131 0 -0.9999>}
 dyn_coat_bk_l_05 {parent=dyn_coat_bk_l_04, loc=<13.6456 0 0>, rot=<0 0 0 1>}
 dyn_coat_bk_l_06 {parent=dyn_coat_bk_l_05, loc=<12.6504 0 0>, rot=<0 0 0 1>}
 dyn_coat_bk_r_MSTR {parent=pelvis, loc=<11.7649 -7.579 8.2398>, rot=<0.1086 -0.5287 0.8411 -0.0341>}
 dyn_coat_bk_r_01 {parent=dyn_coat_bk_r_MSTR, loc=<0.3402 0 0.0001>, rot=<0 0 0 1>}
 dyn_coat_bk_r_02 {parent=dyn_coat_bk_r_01, loc=<8.8284 0 -0.0001>, rot=<0 0 0 1>}
 dyn_coat_bk_r_03 {parent=dyn_coat_bk_r_02, loc=<16.2192 0 0>, rot=<0 0.0178 0 -0.9998>}
 dyn_coat_bk_r_04 {parent=dyn_coat_bk_r_03, loc=<13.6501 0 0>, rot=<0 0.0131 0 -0.9999>}
 dyn_coat_bk_r_05 {parent=dyn_coat_bk_r_04, loc=<13.6456 -0.0001 0>, rot=<0 0 0 1>}
 dyn_coat_bk_r_06 {parent=dyn_coat_bk_r_05, loc=<12.6504 0.0001 -0.0001>, rot=<0 0 0 1>}
 dyn_coat_fr_r_MSTR {parent=pelvis, loc=<16.7171 6.8826 10.1505>, rot=<0.1008 0.1444 0.9842 -0.0178>}
 dyn_coat_fr_r_01 {parent=dyn_coat_fr_r_MSTR, loc=<0.3404 0 0.0001>, rot=<0 0 0 1>}
 dyn_coat_fr_r_02 {parent=dyn_coat_fr_r_01, loc=<10.8406 0 0.6162>, rot=<0 0.0824 0 -0.9966>}
 dyn_coat_fr_r_03 {parent=dyn_coat_fr_r_02, loc=<7.9522 0 0>, rot=<0 0 0 1>}
 dyn_coat_fr_r_04 {parent=dyn_coat_fr_r_03, loc=<16.2736 0.0058 0.0029>, rot=<0 0 0 1>}
 dyn_coat_fr_r_05 {parent=dyn_coat_fr_r_04, loc=<16.9531 -0.0058 -0.0029>, rot=<0 0 0 1>}
 dyn_coat_fr_r_06 {parent=dyn_coat_fr_r_05, loc=<18.3953 0 -0.0001>, rot=<0 0 0 1>}
 dyn_coat_r_MSTR {parent=pelvis, loc=<13.5048 -2.0856 12.4053>, rot=<0.1479 -0.0886 0.985 0.0002>}
 dyn_coat_r_01 {parent=dyn_coat_r_MSTR, loc=<0.3404 0 0.0001>, rot=<0 0 0 1>}
 dyn_coat_r_02 {parent=dyn_coat_r_01, loc=<8.4106 0 -0.0001>, rot=<0 0 0 1>}
 dyn_coat_r_03 {parent=dyn_coat_r_02, loc=<7.8528 0 0>, rot=<0 0 0 1>}
 dyn_coat_r_04 {parent=dyn_coat_r_03, loc=<16.9365 0 0>, rot=<0 0 0 1>}
 dyn_coat_r_05 {parent=dyn_coat_r_04, loc=<17.3764 0 0>, rot=<0 0 0 1>}
 dyn_coat_r_06 {parent=dyn_coat_r_05, loc=<19.0063 0 0>, rot=<0 0 0 1>}
]
variables [
 Alpha_crouching {type=double, default=0}
 Alpha_jump_right {type=double, default=0}
 Alpha_sliding {type=double, default=0}
]
