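# Physics asset for the SkeletonHunterGlitch medium male body, named for skydiving, covering its long coat.
# Pelvis, lower spine, thigh and calf capsules form the body, with seven coat chains hanging from the pelvis via tiny MSTR bodies.
# Chains run from top boxes into capsules that generally fatten toward the hem, with the center back forking in two.
# Free-rotation constraints cross-link neighboring chains row by row, tying the strips into a cloth-like skirt.
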PhysicsAsset /Game/Characters/Player/Male/Medium/Bodies/M_MED_SkeletonHunterGlitch/Meshes/M_MED_SkeletonHunterGlitch_Physics_Skydive
bodies {
 calf_l [{shapeType=Capsule, length=41.9375, radius=7.875, center=<-24.6324 -2.8287 0.0002>, rotation=<rot 90 0 0>}]
 calf_r [{shapeType=Capsule, length=41.9375, radius=7.875, center=<24.6324 2.8287 -0.0002>, rotation=<rot 90 0 0>}]
 dyn_coat_bk_01 [{shapeType=Box, extentX=13.4375, extentY=14.9375, extentZ=1.5625, center=<5.954 0 -0.0003>}]
 dyn_coat_bk_02 [{shapeType=Capsule, length=14, radius=1.5, shapeName=dyn_coat_bk_02_capsule, center=<5.954 0 -0.0003>, rotation=<rot 0 0 90>}]
 dyn_coat_bk_03 [{shapeType=Capsule, length=14, radius=1.5, shapeName=dyn_coat_bk_03_capsule, center=<5.954 0 -0.0003>, rotation=<rot 0 0 90>}]
 dyn_coat_bk_04 [{shapeType=Capsule, length=14, radius=6, shapeName=dyn_coat_bk_04_capsule, center=<5.954 0 -0.0003>, rotation=<rot 0 0 90>}]
 dyn_coat_bk_MSTR [{shapeType=Capsule, length=0.2, radius=0.2}]
 dyn_coat_bk_l_01 [{shapeType=Box, extentX=13.4375, extentY=14.9375, extentZ=1.5625, center=<5.954 0 -0.0003>}]
 dyn_coat_bk_l_02 [{shapeType=Capsule, length=14, radius=1.5, shapeName=dyn_coat_bk_l_02_capsule, center=<5.954 0 -0.0003>, rotation=<rot 0 0 90>}]
 dyn_coat_bk_l_03 [{shapeType=Capsule, length=14, radius=1.5, shapeName=dyn_coat_bk_l_03_capsule, center=<5.954 0 -0.0003>, rotation=<rot 0 0 90>}]
 dyn_coat_bk_l_04 [{shapeType=Capsule, length=22, radius=6, shapeName=dyn_coat_bk_l_04_capsule, center=<5.954 0 -0.0003>, rotation=<rot 0 0 90>}]
 dyn_coat_bk_l_05 [{shapeType=Capsule, length=26.5, radius=6, shapeName=dyn_coat_bk_l_05_capsule, center=<5.954 0 -0.0003>, rotation=<rot 0 0 90>}]
 dyn_coat_bk_l_06 [{shapeType=Capsule, length=26.5, radius=8, shapeName=dyn_coat_bk_l_06_capsule, center=<9.8419 0 -0.0003>, rotation=<rot 0 0 90>}]
 dyn_coat_bk_l_MSTR [{shapeType=Capsule, length=0.2, radius=0.2}]
 dyn_coat_bk_r_01 [{shapeType=Box, extentX=13.4375, extentY=14.9375, extentZ=1.5625, center=<5.954 0 -0.0003>}]
 dyn_coat_bk_r_02 [{shapeType=Capsule, length=14, radius=2.56, shapeName=dyn_coat_bk_r_02_capsule, center=<5.954 0 -0.0003>, rotation=<rot 0 0 90>}]
 dyn_coat_bk_r_03 [{shapeType=Capsule, length=14, radius=2.5, shapeName=dyn_coat_bk_r_03_capsule, center=<5.954 0 -0.0003>, rotation=<rot 0 0 90>}]
 dyn_coat_bk_r_04 [{shapeType=Capsule, length=14, radius=6, shapeName=dyn_coat_bk_r_04_capsule, center=<5.954 0 -0.0003>, rotation=<rot 0 0 90>}]
 dyn_coat_bk_r_05 [{shapeType=Capsule, length=29, radius=6, shapeName=dyn_coat_bk_r_05_capsule, center=<5.954 0 -0.0003>, rotation=<rot 0 0 90>}]
 dyn_coat_bk_r_06 [{shapeType=Capsule, length=29, radius=8, shapeName=dyn_coat_bk_r_06_capsule, center=<12.7623 0.0002 -0.4071>, rotation=<rot 0 0 90>}]
 dyn_coat_bk_r_MSTR [{shapeType=Capsule, length=0.2, radius=0.2}]
 dyn_coat_bkl_05 [{shapeType=Capsule, length=22, radius=6, shapeName=dyn_coat_bkl_05_capsule, center=<5.954 0 -0.0003>, rotation=<rot 0 0 90>}]
 dyn_coat_bkl_06 [{shapeType=Capsule, length=22, radius=7, shapeName=dyn_coat_bkl_06_capsule, center=<5.954 0 -0.0003>, rotation=<rot 0 0 90>}]
 dyn_coat_bkr_05 [{shapeType=Capsule, length=22, radius=6, shapeName=dyn_coat_bkr_05_capsule, center=<5.954 0 -0.0003>, rotation=<rot 0 0 90>}]
 dyn_coat_bkr_06 [{shapeType=Capsule, length=22, radius=7, shapeName=dyn_coat_bkr_06_capsule, center=<5.954 0 -0.0003>, rotation=<rot 0 0 90>}]
 dyn_coat_fr_l_01 [{shapeType=Box, extentX=13.4375, extentY=14.9375, extentZ=1.5625, center=<5.954 0 -0.0003>}]
 dyn_coat_fr_l_02 [{shapeType=Capsule, length=14, radius=1.5, shapeName=dyn_coat_fr_l_02_capsule, center=<5.954 0 -0.0003>, rotation=<rot 0 0 90>}]
 dyn_coat_fr_l_03 [{shapeType=Capsule, length=14, radius=1.5, shapeName=dyn_coat_fr_l_03_capsule, center=<5.954 0 -0.0003>, rotation=<rot 0 0 90>}]
 dyn_coat_fr_l_04 [{shapeType=Capsule, length=21, radius=6, shapeName=dyn_coat_fr_l_04_capsule, center=<5.954 0 -0.0003>, rotation=<rot 0 0 90>}]
 dyn_coat_fr_l_05 [{shapeType=Capsule, length=21, radius=6, shapeName=dyn_coat_fr_l_05_capsule, center=<5.954 0 -0.0003>, rotation=<rot 0 0 90>}]
 dyn_coat_fr_l_06 [{shapeType=Capsule, length=21, radius=6, shapeName=dyn_coat_fr_l_06_capsule, center=<5.954 0 -0.0003>, rotation=<rot 0 0 90>}]
 dyn_coat_fr_l_MSTR [{shapeType=Capsule, length=0.2, radius=0.2}]
 dyn_coat_fr_r_01 [{shapeType=Box, extentX=13.4375, extentY=14.9375, extentZ=1.5625, center=<5.954 0 -0.0003>}]
 dyn_coat_fr_r_02 [{shapeType=Capsule, length=14, radius=1.5, shapeName=dyn_coat_fr_r_02_capsule, center=<5.954 0 -0.0003>, rotation=<rot 0 0 90>}]
 dyn_coat_fr_r_03 [{shapeType=Capsule, length=14, radius=1.5, shapeName=dyn_coat_fr_r_03_capsule, center=<5.954 0 -0.0003>, rotation=<rot 0 0 90>}]
 dyn_coat_fr_r_04 [{shapeType=Capsule, length=14, radius=6, shapeName=dyn_coat_fr_r_04_capsule, center=<5.954 0 -0.0003>, rotation=<rot 0 0 90>}]
 dyn_coat_fr_r_05 [{shapeType=Capsule, length=14, radius=6, shapeName=dyn_coat_fr_r_05_capsule, center=<5.954 0 -0.0003>, rotation=<rot 0 0 90>}]
 dyn_coat_fr_r_06 [{shapeType=Capsule, length=14, radius=6, shapeName=dyn_coat_fr_r_06_capsule, center=<5.954 0 -0.0003>, rotation=<rot 0 0 90>}]
 dyn_coat_fr_r_MSTR [{shapeType=Capsule, length=0.2, radius=0.2}]
 dyn_coat_l_01 [{shapeType=Box, extentX=13.4375, extentY=14.9375, extentZ=1.5625, center=<5.954 0 -0.0003>}]
 dyn_coat_l_02 [{shapeType=Capsule, length=14, radius=1.5, shapeName=dyn_coat_l_02_capsule, center=<5.954 0 -0.0003>, rotation=<rot 0 0 90>}]
 dyn_coat_l_03 [{shapeType=Capsule, length=14, radius=1.5, shapeName=dyn_coat_l_03_capsule, center=<5.954 0 -0.0003>, rotation=<rot 0 0 90>}]
 dyn_coat_l_04 [{shapeType=Capsule, length=21, radius=6, shapeName=dyn_coat_l_04_capsule, center=<5.954 0 -0.0003>, rotation=<rot 0 0 90>}]
 dyn_coat_l_05 [{shapeType=Capsule, length=21, radius=9, shapeName=dyn_coat_l_05_capsule, center=<12.6765 0 0>, rotation=<rot 0 0 90>}]
 dyn_coat_l_06 [{shapeType=Capsule, length=21, radius=9, shapeName=dyn_coat_l_06_capsule, center=<10.2481 0 -0.0004>, rotation=<rot 0 0 90>}]
 dyn_coat_l_MSTR [{shapeType=Capsule, length=0.2, radius=0.2}]
 dyn_coat_r_01 [{shapeType=Box, extentX=13.4375, extentY=14.9375, extentZ=1.5625, center=<5.954 0 -0.0003>}]
 dyn_coat_r_02 [{shapeType=Capsule, length=14, radius=1.5, shapeName=dyn_coat_r_02_capsule, center=<5.954 0 -0.0003>, rotation=<rot 0 0 90>}]
 dyn_coat_r_03 [{shapeType=Capsule, length=15, radius=1.5, shapeName=dyn_coat_r_03_capsule, center=<5.954 0 -0.0003>, rotation=<rot 0 0 90>}]
 dyn_coat_r_04 [{shapeType=Capsule, length=14, radius=6, shapeName=dyn_coat_r_04_capsule, center=<5.954 0 -0.0003>, rotation=<rot 0 0 90>}]
 dyn_coat_r_05 [{shapeType=Capsule, length=14, radius=9.25, shapeName=dyn_coat_r_05_capsule, center=<10.616 0 -0.0004>, rotation=<rot 0 0 90>}]
 dyn_coat_r_06 [{shapeType=Capsule, length=23, radius=11.25, shapeName=dyn_coat_r_06_capsule, center=<10.8961 0.0001 2.1803>, rotation=<rot 0 0 90>}]
 dyn_coat_r_MSTR [{shapeType=Capsule, length=0.2, radius=0.2}]
 pelvis [{shapeType=Capsule, length=9.5938, radius=10.4062, center=<-2.3213 0.0139 0>}]
 spine_01 [{shapeType=Capsule, length=6.8438, radius=10, center=<0 0.559 0.0001>}]
 spine_02 [{shapeType=Capsule, length=6.125, radius=8.2188, center=<-0.0387 0.9601 0>}]
 thigh_l [{shapeType=Capsule, length=33.4688, radius=8.5938, center=<-19.2427 0.0018 0>, rotation=<rot 90 0 0>}]
 thigh_r [
  {shapeType=Capsule, length=33.4688, radius=8.5938, center=<19.2427 -0.0018 0>, rotation=<rot 90 0 0>}
  {shapeType=Capsule, length=10.9062, radius=5.1875, center=<12.2163 -8.2543 3.4478>, rotation=<rot 87.6801 -151.7164 -145.7971>}
 ]
}
constraints [
 {bone1Name=dyn_coat_bk_MSTR, bone2Name=pelvis, swing1LimitDegrees=45, swing2LimitDegrees=45, twistLimitDegrees=5}
 {bone1Name=dyn_coat_bk_01, bone2Name=dyn_coat_bk_MSTR, swing1LimitDegrees=45, swing2LimitDegrees=45, twistLimitDegrees=5}
 {bone1Name=dyn_coat_bk_02, bone2Name=dyn_coat_bk_01, swing1LimitDegrees=45, swing2LimitDegrees=45, twistLimitDegrees=5}
 {
  bone1Name=dyn_coat_bk_r_02
  bone2Name=dyn_coat_bk_02
  swing1LimitDegrees=45
  swing1Motion=Free
  swing2LimitDegrees=45
  swing2Motion=Free
  twistLimitDegrees=45
  twistMotion=Free
 }
 {
  bone1Name=dyn_coat_bk_l_02
  bone2Name=dyn_coat_bk_02
  swing1LimitDegrees=45
  swing1Motion=Free
  swing2LimitDegrees=45
  swing2Motion=Free
  twistLimitDegrees=45
  twistMotion=Free
 }
 {bone1Name=dyn_coat_bk_03, bone2Name=dyn_coat_bk_02, swing1LimitDegrees=45, swing2LimitDegrees=45, twistLimitDegrees=5}
 {
  bone1Name=dyn_coat_bk_r_03
  bone2Name=dyn_coat_bk_03
  swing1LimitDegrees=45
  swing1Motion=Free
  swing2LimitDegrees=45
  swing2Motion=Free
  twistLimitDegrees=45
  twistMotion=Free
 }
 {
  bone1Name=dyn_coat_bk_l_03
  bone2Name=dyn_coat_bk_03
  swing1LimitDegrees=45
  swing1Motion=Free
  swing2LimitDegrees=45
  swing2Motion=Free
  twistLimitDegrees=45
  twistMotion=Free
 }
 {bone1Name=dyn_coat_bk_04, bone2Name=dyn_coat_bk_03, swing1LimitDegrees=45, swing2LimitDegrees=45, twistLimitDegrees=5}
 {
  bone1Name=dyn_coat_bk_r_04
  bone2Name=dyn_coat_bk_04
  swing1LimitDegrees=45
  swing1Motion=Free
  swing2LimitDegrees=45
  swing2Motion=Free
  twistLimitDegrees=45
  twistMotion=Free
 }
 {
  bone1Name=dyn_coat_bk_l_04
  bone2Name=dyn_coat_bk_04
  swing1LimitDegrees=45
  swing1Motion=Free
  swing2LimitDegrees=45
  swing2Motion=Free
  twistLimitDegrees=45
  twistMotion=Free
 }
 {bone1Name=dyn_coat_bkl_05, bone2Name=dyn_coat_bk_04, swing1LimitDegrees=45, swing2LimitDegrees=45, twistLimitDegrees=5}
 {
  bone1Name=dyn_coat_bk_l_05
  bone2Name=dyn_coat_bkl_05
  swing1LimitDegrees=45
  swing1Motion=Free
  swing2LimitDegrees=45
  swing2Motion=Free
  twistLimitDegrees=45
  twistMotion=Free
 }
 {bone1Name=dyn_coat_bkl_06, bone2Name=dyn_coat_bkl_05, swing1LimitDegrees=45, swing2LimitDegrees=45, twistLimitDegrees=5}
 {
  bone1Name=dyn_coat_bk_l_06
  bone2Name=dyn_coat_bkl_06
  swing1LimitDegrees=45
  swing1Motion=Free
  swing2LimitDegrees=45
  swing2Motion=Free
  twistLimitDegrees=45
  twistMotion=Free
 }
 {bone1Name=dyn_coat_bkr_05, bone2Name=dyn_coat_bk_04, swing1LimitDegrees=45, swing2LimitDegrees=45, twistLimitDegrees=5}
 {
  bone1Name=dyn_coat_bk_r_05
  bone2Name=dyn_coat_bkr_05
  swing1LimitDegrees=45
  swing1Motion=Free
  swing2LimitDegrees=45
  swing2Motion=Free
  twistLimitDegrees=45
  twistMotion=Free
 }
 {bone1Name=dyn_coat_bkr_06, bone2Name=dyn_coat_bkr_05, swing1LimitDegrees=45, swing2LimitDegrees=45, twistLimitDegrees=5}
 {
  bone1Name=dyn_coat_bk_r_06
  bone2Name=dyn_coat_bkr_06
  swing1LimitDegrees=45
  swing1Motion=Free
  swing2LimitDegrees=45
  swing2Motion=Free
  twistLimitDegrees=45
  twistMotion=Free
 }
 {
  bone1Name=dyn_coat_bk_l_MSTR
  bone2Name=pelvis
  swing1LimitDegrees=45
  swing1Motion=Free
  swing2LimitDegrees=45
  swing2Motion=Free
  twistLimitDegrees=45
  twistMotion=Free
 }
 {bone1Name=dyn_coat_bk_l_01, bone2Name=dyn_coat_bk_l_MSTR, swing1LimitDegrees=45, swing2LimitDegrees=45, twistLimitDegrees=5}
 {bone1Name=dyn_coat_bk_l_02, bone2Name=dyn_coat_bk_l_01, swing1LimitDegrees=45, swing2LimitDegrees=45, twistLimitDegrees=5}
 {
  bone1Name=dyn_coat_l_02
  bone2Name=dyn_coat_bk_l_02
  swing1LimitDegrees=45
  swing1Motion=Free
  swing2LimitDegrees=45
  swing2Motion=Free
  twistLimitDegrees=45
  twistMotion=Free
 }
 {bone1Name=dyn_coat_bk_l_03, bone2Name=dyn_coat_bk_l_02, swing1LimitDegrees=45, swing2LimitDegrees=45, twistLimitDegrees=5}
 {
  bone1Name=dyn_coat_l_03
  bone2Name=dyn_coat_bk_l_03
  swing1LimitDegrees=45
  swing1Motion=Free
  swing2LimitDegrees=45
  swing2Motion=Free
  twistLimitDegrees=45
  twistMotion=Free
 }
 {bone1Name=dyn_coat_bk_l_04, bone2Name=dyn_coat_bk_l_03, swing1LimitDegrees=45, swing2LimitDegrees=45, twistLimitDegrees=5}
 {
  bone1Name=dyn_coat_l_04
  bone2Name=dyn_coat_bk_l_04
  swing1LimitDegrees=45
  swing1Motion=Free
  swing2LimitDegrees=45
  swing2Motion=Free
  twistLimitDegrees=45
  twistMotion=Free
 }
 {bone1Name=dyn_coat_bk_l_05, bone2Name=dyn_coat_bk_l_04, swing1LimitDegrees=45, swing2LimitDegrees=45, twistLimitDegrees=5}
 {
  bone1Name=dyn_coat_l_05
  bone2Name=dyn_coat_bk_l_05
  swing1LimitDegrees=45
  swing1Motion=Free
  swing2LimitDegrees=45
  swing2Motion=Free
  twistLimitDegrees=45
  twistMotion=Free
 }
 {bone1Name=dyn_coat_bk_l_06, bone2Name=dyn_coat_bk_l_05, swing1LimitDegrees=45, swing2LimitDegrees=45, twistLimitDegrees=5}
 {
  bone1Name=dyn_coat_l_06
  bone2Name=dyn_coat_bk_l_06
  swing1LimitDegrees=45
  swing1Motion=Free
  swing2LimitDegrees=45
  swing2Motion=Free
  twistLimitDegrees=45
  twistMotion=Free
 }
 {
  bone1Name=dyn_coat_bk_r_MSTR
  bone2Name=pelvis
  swing1LimitDegrees=45
  swing1Motion=Free
  swing2LimitDegrees=45
  swing2Motion=Free
  twistLimitDegrees=45
  twistMotion=Free
 }
 {bone1Name=dyn_coat_bk_r_01, bone2Name=dyn_coat_bk_r_MSTR, swing1LimitDegrees=45, swing2LimitDegrees=45, twistLimitDegrees=5}
 {bone1Name=dyn_coat_bk_r_02, bone2Name=dyn_coat_bk_r_01, swing1LimitDegrees=45, swing2LimitDegrees=45, twistLimitDegrees=5}
 {
  bone1Name=dyn_coat_r_02
  bone2Name=dyn_coat_bk_r_02
  swing1LimitDegrees=45
  swing1Motion=Free
  swing2LimitDegrees=45
  swing2Motion=Free
  twistLimitDegrees=45
  twistMotion=Free
 }
 {bone1Name=dyn_coat_bk_r_03, bone2Name=dyn_coat_bk_r_02, swing1LimitDegrees=45, swing2LimitDegrees=45, twistLimitDegrees=5}
 {
  bone1Name=dyn_coat_r_03
  bone2Name=dyn_coat_bk_r_03
  swing1LimitDegrees=45
  swing1Motion=Free
  swing2LimitDegrees=45
  swing2Motion=Free
  twistLimitDegrees=45
  twistMotion=Free
 }
 {bone1Name=dyn_coat_bk_r_04, bone2Name=dyn_coat_bk_r_03, swing1LimitDegrees=45, swing2LimitDegrees=45, twistLimitDegrees=5}
 {
  bone1Name=dyn_coat_r_04
  bone2Name=dyn_coat_bk_r_04
  swing1LimitDegrees=45
  swing1Motion=Free
  swing2LimitDegrees=45
  swing2Motion=Free
  twistLimitDegrees=45
  twistMotion=Free
 }
 {bone1Name=dyn_coat_bk_r_05, bone2Name=dyn_coat_bk_r_04, swing1LimitDegrees=45, swing2LimitDegrees=45, twistLimitDegrees=5}
 {
  bone1Name=dyn_coat_r_05
  bone2Name=dyn_coat_bk_r_05
  swing1LimitDegrees=45
  swing1Motion=Free
  swing2LimitDegrees=45
  swing2Motion=Free
  twistLimitDegrees=45
  twistMotion=Free
 }
 {bone1Name=dyn_coat_bk_r_06, bone2Name=dyn_coat_bk_r_05, swing1LimitDegrees=45, swing2LimitDegrees=45, twistLimitDegrees=5}
 {
  bone1Name=dyn_coat_r_06
  bone2Name=dyn_coat_bk_r_06
  swing1LimitDegrees=45
  swing1Motion=Free
  swing2LimitDegrees=45
  swing2Motion=Free
  twistLimitDegrees=45
  twistMotion=Free
 }
 {bone1Name=dyn_coat_fr_l_MSTR, bone2Name=pelvis, swing1LimitDegrees=45, swing2LimitDegrees=45, twistLimitDegrees=5}
 {bone1Name=dyn_coat_fr_l_01, bone2Name=dyn_coat_fr_l_MSTR, swing1LimitDegrees=45, swing2LimitDegrees=45, twistLimitDegrees=5}
 {bone1Name=dyn_coat_fr_l_02, bone2Name=dyn_coat_fr_l_01, swing1LimitDegrees=45, swing2LimitDegrees=45, twistLimitDegrees=5}
 {
  bone1Name=dyn_coat_fr_l_02
  bone2Name=dyn_coat_l_02
  swing1LimitDegrees=45
  swing1Motion=Free
  swing2LimitDegrees=45
  swing2Motion=Free
  twistLimitDegrees=45
  twistMotion=Free
 }
 {bone1Name=dyn_coat_fr_l_03, bone2Name=dyn_coat_fr_l_02, swing1LimitDegrees=45, swing2LimitDegrees=45, twistLimitDegrees=5}
 {
  bone1Name=dyn_coat_fr_l_03
  bone2Name=dyn_coat_l_03
  swing1LimitDegrees=45
  swing1Motion=Free
  swing2LimitDegrees=45
  swing2Motion=Free
  twistLimitDegrees=45
  twistMotion=Free
 }
 {bone1Name=dyn_coat_fr_l_04, bone2Name=dyn_coat_fr_l_03, swing1LimitDegrees=45, swing2LimitDegrees=45, twistLimitDegrees=5}
 {
  bone1Name=dyn_coat_fr_l_04
  bone2Name=dyn_coat_l_04
  swing1LimitDegrees=45
  swing1Motion=Free
  swing2LimitDegrees=45
  swing2Motion=Free
  twistLimitDegrees=45
  twistMotion=Free
 }
 {bone1Name=dyn_coat_fr_l_05, bone2Name=dyn_coat_fr_l_04, swing1LimitDegrees=45, swing2LimitDegrees=45, twistLimitDegrees=5}
 {
  bone1Name=dyn_coat_fr_l_05
  bone2Name=dyn_coat_l_05
  swing1LimitDegrees=45
  swing1Motion=Free
  swing2LimitDegrees=45
  swing2Motion=Free
  twistLimitDegrees=45
  twistMotion=Free
 }
 {bone1Name=dyn_coat_fr_l_06, bone2Name=dyn_coat_fr_l_05, swing1LimitDegrees=45, swing2LimitDegrees=45, twistLimitDegrees=5}
 {
  bone1Name=dyn_coat_fr_l_06
  bone2Name=dyn_coat_l_06
  swing1LimitDegrees=45
  swing1Motion=Free
  swing2LimitDegrees=45
  swing2Motion=Free
  twistLimitDegrees=45
  twistMotion=Free
 }
 {
  bone1Name=dyn_coat_fr_r_MSTR
  bone2Name=pelvis
  swing1LimitDegrees=45
  swing1Motion=Free
  swing2LimitDegrees=45
  swing2Motion=Free
  twistLimitDegrees=45
  twistMotion=Free
 }
 {bone1Name=dyn_coat_fr_r_01, bone2Name=dyn_coat_fr_r_MSTR, swing1LimitDegrees=45, swing2LimitDegrees=45, twistLimitDegrees=5}
 {bone1Name=dyn_coat_fr_r_02, bone2Name=dyn_coat_fr_r_01, swing1LimitDegrees=45, swing2LimitDegrees=45, twistLimitDegrees=5}
 {
  bone1Name=dyn_coat_fr_r_02
  bone2Name=dyn_coat_r_02
  swing1LimitDegrees=45
  swing1Motion=Free
  swing2LimitDegrees=45
  swing2Motion=Free
  twistLimitDegrees=45
  twistMotion=Free
 }
 {bone1Name=dyn_coat_fr_r_03, bone2Name=dyn_coat_fr_r_02, swing1LimitDegrees=45, swing2LimitDegrees=45, twistLimitDegrees=5}
 {
  bone1Name=dyn_coat_fr_r_03
  bone2Name=dyn_coat_r_03
  swing1LimitDegrees=45
  swing1Motion=Free
  swing2LimitDegrees=45
  swing2Motion=Free
  twistLimitDegrees=45
  twistMotion=Free
 }
 {bone1Name=dyn_coat_fr_r_04, bone2Name=dyn_coat_fr_r_03, swing1LimitDegrees=45, swing2LimitDegrees=45, twistLimitDegrees=5}
 {
  bone1Name=dyn_coat_fr_r_04
  bone2Name=dyn_coat_r_04
  swing1LimitDegrees=45
  swing1Motion=Free
  swing2LimitDegrees=45
  swing2Motion=Free
  twistLimitDegrees=45
  twistMotion=Free
 }
 {bone1Name=dyn_coat_fr_r_05, bone2Name=dyn_coat_fr_r_04, swing1LimitDegrees=45, swing2LimitDegrees=45, twistLimitDegrees=5}
 {
  bone1Name=dyn_coat_fr_r_05
  bone2Name=dyn_coat_r_05
  swing1LimitDegrees=45
  swing1Motion=Free
  swing2LimitDegrees=45
  swing2Motion=Free
  twistLimitDegrees=45
  twistMotion=Free
 }
 {bone1Name=dyn_coat_fr_r_06, bone2Name=dyn_coat_fr_r_05, swing1LimitDegrees=45, swing2LimitDegrees=45, twistLimitDegrees=5}
 {
  bone1Name=dyn_coat_fr_r_06
  bone2Name=dyn_coat_r_06
  swing1LimitDegrees=45
  swing1Motion=Free
  swing2LimitDegrees=45
  swing2Motion=Free
  twistLimitDegrees=45
  twistMotion=Free
 }
 {bone1Name=dyn_coat_l_MSTR, bone2Name=pelvis, swing1LimitDegrees=45, swing2LimitDegrees=45, twistLimitDegrees=5}
 {bone1Name=dyn_coat_l_01, bone2Name=dyn_coat_l_MSTR, swing1LimitDegrees=45, swing2LimitDegrees=45, twistLimitDegrees=5}
 {bone1Name=dyn_coat_l_02, bone2Name=dyn_coat_l_01, swing1LimitDegrees=45, swing2LimitDegrees=45, twistLimitDegrees=5}
 {bone1Name=dyn_coat_l_03, bone2Name=dyn_coat_l_02, swing1LimitDegrees=45, swing2LimitDegrees=45, twistLimitDegrees=5}
 {bone1Name=dyn_coat_l_04, bone2Name=dyn_coat_l_03, swing1LimitDegrees=45, swing2LimitDegrees=45, twistLimitDegrees=5}
 {bone1Name=dyn_coat_l_05, bone2Name=dyn_coat_l_04, swing1LimitDegrees=45, swing2LimitDegrees=45, twistLimitDegrees=5}
 {bone1Name=dyn_coat_l_06, bone2Name=dyn_coat_l_05, swing1LimitDegrees=45, swing2LimitDegrees=45, twistLimitDegrees=5}
 {
  bone1Name=dyn_coat_r_MSTR
  bone2Name=pelvis
  swing1LimitDegrees=45
  swing1Motion=Free
  swing2LimitDegrees=45
  swing2Motion=Free
  twistLimitDegrees=45
  twistMotion=Free
 }
 {bone1Name=dyn_coat_r_01, bone2Name=dyn_coat_r_MSTR, swing1LimitDegrees=45, swing2LimitDegrees=45, twistLimitDegrees=5}
 {bone1Name=dyn_coat_r_02, bone2Name=dyn_coat_r_01, swing1LimitDegrees=45, swing2LimitDegrees=45, twistLimitDegrees=5}
 {bone1Name=dyn_coat_r_03, bone2Name=dyn_coat_r_02, swing1LimitDegrees=45, swing2LimitDegrees=45, twistLimitDegrees=5}
 {bone1Name=dyn_coat_r_04, bone2Name=dyn_coat_r_03, swing1LimitDegrees=45, swing2LimitDegrees=45, twistLimitDegrees=5}
 {bone1Name=dyn_coat_r_05, bone2Name=dyn_coat_r_04, swing1LimitDegrees=45, swing2LimitDegrees=45, twistLimitDegrees=5}
 {bone1Name=dyn_coat_r_06, bone2Name=dyn_coat_r_05, swing1LimitDegrees=45, swing2LimitDegrees=45, twistLimitDegrees=5}
 {bone1Name=spine_01, bone2Name=pelvis, swing1LimitDegrees=45, swing2LimitDegrees=45, twistLimitDegrees=5}
 {bone1Name=spine_02, bone2Name=spine_01, swing1LimitDegrees=45, swing2LimitDegrees=45, twistLimitDegrees=5}
 {
  bone1Name=thigh_l
  bone2Name=pelvis
  swing1LimitDegrees=45
  swing1Motion=Free
  swing2LimitDegrees=45
  swing2Motion=Free
  twistLimitDegrees=45
  twistMotion=Free
 }
 {
  bone1Name=calf_l
  bone2Name=thigh_l
  swing1LimitDegrees=45
  swing1Motion=Free
  swing2LimitDegrees=45
  swing2Motion=Free
  twistLimitDegrees=45
  twistMotion=Free
 }
 {
  bone1Name=thigh_r
  bone2Name=pelvis
  swing1LimitDegrees=45
  swing1Motion=Free
  swing2LimitDegrees=45
  swing2Motion=Free
  twistLimitDegrees=45
  twistMotion=Free
 }
 {
  bone1Name=calf_r
  bone2Name=thigh_r
  swing1LimitDegrees=45
  swing1Motion=Free
  swing2LimitDegrees=45
  swing2Motion=Free
  twistLimitDegrees=45
  twistMotion=Free
 }
]
