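# Post-process-style AnimBlueprint for the SkeletonHunterGlitch player body, animating its long coat over the pose copied from the parent mesh.
# In-vehicle coat-bone scaling and two Control Rigs (RBAN on/off) feed a cached main pose.
# RigidBody sims with state-specific settings cover standing, crouching, skydiving and emotes.
# With rigid bodies disabled, a cached fallback with crouch offsets replaces simulation; constraint chains tie coat bones to pelvis and legs.
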AnimBlueprint /Game/Characters/Player/Male/Medium/Bodies/M_MED_SkeletonHunterGlitch/Meshes/M_MED_SkeletonHunterGlitch_AnimBP
target_skeleton=&/M_MED_SkeletonHunterGlitch_Skeleton
graphs [
 AnimGraph {nodes [
   n1 = "Constraint - dyn_jacketSide_mstr_l (P)"(Alpha=1, ConstraintWeights_0=1) {
    alphaBoolBlend {blendOption=Linear}
    boneToModify {boneName=dyn_jacketSide_mstr_l}
    constraintSetup [{targetBone {boneName=spine_02}}]
    constraintWeights [1]
   }
   n2 = "Constraint - dyn_jacketBack_mstr_l (P)"(ComponentPose=@n1, Alpha=1, ConstraintWeights_0=1) {
    alphaBoolBlend {blendOption=Linear}
    boneToModify {boneName=dyn_jacketBack_mstr_l}
    constraintSetup [{targetBone {boneName=spine_02}}]
    constraintWeights [1]
   }
   n3 = "Constraint - dyn_jacketBack_mstr_r (P)"(ComponentPose=@n2, Alpha=1, ConstraintWeights_0=1) {
    alphaBoolBlend {blendOption=Linear}
    boneToModify {boneName=dyn_jacketBack_mstr_r}
    constraintSetup [{targetBone {boneName=spine_02}}]
    constraintWeights [1]
   }
   "Constraint - dyn_jacketSide_mstr_r (P)"(ComponentPose=@n3, Alpha=1, ConstraintWeights_0=1) {
    alphaBoolBlend {blendOption=Linear}
    boneToModify {boneName=dyn_jacketSide_mstr_r}
    constraintSetup [{targetBone {boneName=spine_02}}]
    constraintWeights [1]
   }
   n4 = "Constraint - dyn_coat_fr_l_02 (PP)"!(
    ComponentPose=Local To Component(LocalPose=Use cached pose 'main_SavedPose'())
    Alpha=1
    ConstraintWeights_0=0.2
    ConstraintWeights_1:float=0.8
   ) -> (Pose) {
    alphaBoolBlend {blendOption=Linear}
    boneToModify {boneName=dyn_coat_fr_l_02}
    constraintSetup [{targetBone {boneName=thigh_l}}, {targetBone {boneName=pelvis}}]
    constraintWeights [1,1]
   }
   n5 = "Constraint - dyn_coat_l_02 (PP)"!(ComponentPose=@n4, Alpha=1, ConstraintWeights_0=0.2, ConstraintWeights_1:float=0.8) -> (Pose) {
    alphaBoolBlend {blendOption=Linear}
    boneToModify {boneName=dyn_coat_l_02}
    constraintSetup [{targetBone {boneName=thigh_l}}, {targetBone {boneName=pelvis}}]
    constraintWeights [1,1]
   }
   n6 = "Constraint - dyn_coat_bk_l_02 (PP)"!(ComponentPose=@n5, Alpha=1, ConstraintWeights_0=0.2, ConstraintWeights_1:float=0.8) -> (Pose) {
    alphaBoolBlend {blendOption=Linear}
    boneToModify {boneName=dyn_coat_bk_l_02}
    constraintSetup [{targetBone {boneName=thigh_l}}, {targetBone {boneName=pelvis}}]
    constraintWeights [1,1]
   }
   n7 = "Constraint - dyn_coat_fr_l_03 (PP)"!(ComponentPose=@n6, Alpha=1, ConstraintWeights_0=0.5, ConstraintWeights_1:float=0.5) -> (Pose) {
    alphaBoolBlend {blendOption=Linear}
    boneToModify {boneName=dyn_coat_fr_l_03}
    constraintSetup [{targetBone {boneName=thigh_l}}, {targetBone {boneName=pelvis}}]
    constraintWeights [1,1]
   }
   n8 = "Constraint - dyn_coat_l_03 (PP)"!(ComponentPose=@n7, Alpha=1, ConstraintWeights_0=0.5, ConstraintWeights_1:float=0.5) -> (Pose) {
    alphaBoolBlend {blendOption=Linear}
    boneToModify {boneName=dyn_coat_l_03}
    constraintSetup [{targetBone {boneName=thigh_l}}, {targetBone {boneName=pelvis}}]
    constraintWeights [1,1]
   }
   n9 = "Constraint - dyn_coat_bk_l_03 (PP)"!(ComponentPose=@n8, Alpha=1, ConstraintWeights_0=0.5, ConstraintWeights_1:float=0.5) -> (Pose) {
    alphaBoolBlend {blendOption=Linear}
    boneToModify {boneName=dyn_coat_bk_l_03}
    constraintSetup [{targetBone {boneName=thigh_l}}, {targetBone {boneName=pelvis}}]
    constraintWeights [1,1]
   }
   n10 = "Constraint - dyn_coat_fr_l_05 (PP)"!(ComponentPose=@n9, Alpha=1, ConstraintWeights_0=0.5, ConstraintWeights_1:float=0.5) -> (Pose) {
    alphaBoolBlend {blendOption=Linear}
    boneToModify {boneName=dyn_coat_fr_l_05}
    constraintSetup [{targetBone {boneName=thigh_l}}, {targetBone {boneName=calf_l}}]
    constraintWeights [1,1]
   }
   n11 = "Constraint - dyn_coat_l_05 (PP)"!(ComponentPose=@n10, Alpha=1, ConstraintWeights_0=0.5, ConstraintWeights_1:float=0.5) -> (Pose) {
    alphaBoolBlend {blendOption=Linear}
    boneToModify {boneName=dyn_coat_l_05}
    constraintSetup [{targetBone {boneName=thigh_l}}, {targetBone {boneName=calf_l}}]
    constraintWeights [1,1]
   }
   n12 = "Constraint - dyn_coat_bk_l_05 (PP)"!(ComponentPose=@n11, Alpha=1, ConstraintWeights_0=0.5, ConstraintWeights_1:float=0.5) -> (Pose) {
    alphaBoolBlend {blendOption=Linear}
    boneToModify {boneName=dyn_coat_bk_l_05}
    constraintSetup [{targetBone {boneName=thigh_l}}, {targetBone {boneName=calf_l}}]
    constraintWeights [1,1]
   }
   n13 = "Constraint - dyn_coat_bk_l_06 (PP)"!(ComponentPose=@n12, Alpha=1, ConstraintWeights_0=0.25, ConstraintWeights_1:float=0.75) -> (Pose) {
    alphaBoolBlend {blendOption=Linear}
    boneToModify {boneName=dyn_coat_bk_l_06}
    constraintSetup [{targetBone {boneName=thigh_l}}, {targetBone {boneName=calf_l}}]
    constraintWeights [1,1]
   }
   n14 = "Constraint - dyn_coat_fr_r_02 (PP)"!(ComponentPose=@n13, Alpha=1, ConstraintWeights_0=0.2, ConstraintWeights_1:float=0.8) -> (Pose) {
    alphaBoolBlend {blendOption=Linear}
    boneToModify {boneName=dyn_coat_fr_r_02}
    constraintSetup [{targetBone {boneName=thigh_r}}, {targetBone {boneName=pelvis}}]
    constraintWeights [1,1]
   }
   n15 = "Constraint - dyn_coat_r_02 (PP)"!(ComponentPose=@n14, Alpha=1, ConstraintWeights_0=0.2, ConstraintWeights_1:float=0.8) -> (Pose) {
    alphaBoolBlend {blendOption=Linear}
    boneToModify {boneName=dyn_coat_r_02}
    constraintSetup [{targetBone {boneName=thigh_r}}, {targetBone {boneName=pelvis}}]
    constraintWeights [1,1]
   }
   n16 = "Constraint - dyn_coat_bk_r_02 (PP)"!(ComponentPose=@n15, Alpha=1, ConstraintWeights_0=0.2, ConstraintWeights_1:float=0.8) -> (Pose) {
    alphaBoolBlend {blendOption=Linear}
    boneToModify {boneName=dyn_coat_bk_r_02}
    constraintSetup [{targetBone {boneName=thigh_r}}, {targetBone {boneName=pelvis}}]
    constraintWeights [1,1]
   }
   n17 = "Constraint - dyn_coat_fr_r_03 (PP)"!(ComponentPose=@n16, Alpha=1, ConstraintWeights_0=0.5, ConstraintWeights_1:float=0.5) -> (Pose) {
    alphaBoolBlend {blendOption=Linear}
    boneToModify {boneName=dyn_coat_fr_r_03}
    constraintSetup [{targetBone {boneName=thigh_r}}, {targetBone {boneName=pelvis}}]
    constraintWeights [1,1]
   }
   n18 = "Constraint - dyn_coat_r_03 (PP)"!(ComponentPose=@n17, Alpha=1, ConstraintWeights_0=0.5, ConstraintWeights_1:float=0.5) -> (Pose) {
    alphaBoolBlend {blendOption=Linear}
    boneToModify {boneName=dyn_coat_r_03}
    constraintSetup [{targetBone {boneName=thigh_r}}, {targetBone {boneName=pelvis}}]
    constraintWeights [1,1]
   }
   n19 = "Constraint - dyn_coat_bk_r_03 (PP)"!(ComponentPose=@n18, Alpha=1, ConstraintWeights_0=0.5, ConstraintWeights_1:float=0.5) -> (Pose) {
    alphaBoolBlend {blendOption=Linear}
    boneToModify {boneName=dyn_coat_bk_r_03}
    constraintSetup [{targetBone {boneName=thigh_r}}, {targetBone {boneName=pelvis}}]
    constraintWeights [1,1]
   }
   n20 = "Constraint - dyn_coat_fr_r_05 (PP)"!(ComponentPose=@n19, Alpha=1, ConstraintWeights_0=0.5, ConstraintWeights_1:float=0.5) -> (Pose) {
    alphaBoolBlend {blendOption=Linear}
    boneToModify {boneName=dyn_coat_fr_r_05}
    constraintSetup [{targetBone {boneName=thigh_r}}, {targetBone {boneName=calf_r}}]
    constraintWeights [1,1]
   }
   n21 = "Constraint - dyn_coat_r_05 (PP)"!(ComponentPose=@n20, Alpha=1, ConstraintWeights_0=0.5, ConstraintWeights_1:float=0.5) -> (Pose) {
    alphaBoolBlend {blendOption=Linear}
    boneToModify {boneName=dyn_coat_r_05}
    constraintSetup [{targetBone {boneName=thigh_r}}, {targetBone {boneName=calf_r}}]
    constraintWeights [1,1]
   }
   n22 = "Constraint - dyn_coat_bk_r_05 (PP)"!(ComponentPose=@n21, Alpha=1, ConstraintWeights_0=0.5, ConstraintWeights_1:float=0.5) -> (Pose) {
    alphaBoolBlend {blendOption=Linear}
    boneToModify {boneName=dyn_coat_bk_r_05}
    constraintSetup [{targetBone {boneName=thigh_r}}, {targetBone {boneName=calf_r}}]
    constraintWeights [1,1]
   }
   n23 = "Constraint - dyn_coat_bk_r_06 (PP)"!(ComponentPose=@n22, Alpha=1, ConstraintWeights_0=0.25, ConstraintWeights_1:float=0.75) -> (Pose) {
    alphaBoolBlend {blendOption=Linear}
    boneToModify {boneName=dyn_coat_bk_r_06}
    constraintSetup [{targetBone {boneName=thigh_r}}, {targetBone {boneName=calf_r}}]
    constraintWeights [1,1]
   }
   n24 = "Constraint - dyn_coat_bk_02 (PP)"!(ComponentPose=@n23, Alpha=1, ConstraintWeights_0=0.5, ConstraintWeights_1:float=0.5) -> (Pose) {
    alphaBoolBlend {blendOption=Linear}
    boneToModify {boneName=dyn_coat_bk_02}
    constraintSetup [{targetBone {boneName=dyn_coat_bk_l_02}}, {targetBone {boneName=dyn_coat_bk_r_02}}]
    constraintWeights [1,1]
   }
   n25 = "Constraint - dyn_coat_bk_03 (PP)"!(ComponentPose=@n24, Alpha=1, ConstraintWeights_0=0.5, ConstraintWeights_1:float=0.5) -> (Pose) {
    alphaBoolBlend {blendOption=Linear}
    boneToModify {boneName=dyn_coat_bk_03}
    constraintSetup [{targetBone {boneName=dyn_coat_bk_l_03}}, {targetBone {boneName=dyn_coat_bk_r_03}}]
    constraintWeights [1,1]
   }
   n26 = "Constraint - dyn_coat_bkl_05 (PP)"!(ComponentPose=@n25, Alpha=1, ConstraintWeights_0=0.6, ConstraintWeights_1:float=0.4) -> (Pose) {
    alphaBoolBlend {blendOption=Linear}
    boneToModify {boneName=dyn_coat_bkl_05}
    constraintSetup [{targetBone {boneName=dyn_coat_bk_l_05}}, {targetBone {boneName=dyn_coat_bk_r_05}}]
    constraintWeights [1,1]
   }
   n27 = "Constraint - dyn_coat_bkr_05 (PP)"!(ComponentPose=@n26, Alpha=1, ConstraintWeights_0=0.4, ConstraintWeights_1:float=0.6) -> (Pose) {
    alphaBoolBlend {blendOption=Linear}
    boneToModify {boneName=dyn_coat_bkr_05}
    constraintSetup [{targetBone {boneName=dyn_coat_bk_l_05}}, {targetBone {boneName=dyn_coat_bk_r_05}}]
    constraintWeights [1,1]
   }
   n28 = "Constraint - dyn_coat_bkl_06 (PP)"!(ComponentPose=@n27, Alpha=1, ConstraintWeights_0=0.6, ConstraintWeights_1:float=0.4) -> (Pose) {
    alphaBoolBlend {blendOption=Linear}
    boneToModify {boneName=dyn_coat_bkl_06}
    constraintSetup [{targetBone {boneName=dyn_coat_bk_l_06}}, {targetBone {boneName=dyn_coat_bk_r_06}}]
    constraintWeights [1,1]
   }
   "Constraint - dyn_coat_bkr_06 (PP)"!(ComponentPose=@n28, Alpha=1, ConstraintWeights_0=0.4, ConstraintWeights_1:float=0.6) -> (Pose) {
    alphaBoolBlend {blendOption=Linear}
    boneToModify {boneName=dyn_coat_bkr_06}
    constraintSetup [{targetBone {boneName=dyn_coat_bk_l_06}}, {targetBone {boneName=dyn_coat_bk_r_06}}]
    constraintWeights [1,1]
   }
   n29 = Get IsNotCrouchRunning() -> (IsNotCrouchRunning:float)
   "Transform (Modify) Bone - Bone: dyn_jacketBack_03_l"(Alpha=@n29, Translation="0,0,0", Rotation="25,0,0.000", Scale="1,1,1") {
    rotationMode=BMM_Additive
    rotationSpace=BCS_BoneSpace
    alphaBoolBlend {blendOption=Linear}
    alphaScaleBiasClamp {bInterpResult=true, interpSpeedDecreasing=2, interpSpeedIncreasing=5}
    boneToModify {boneName=dyn_jacketBack_03_l}
    rotation {pitch=25}
   }
   "Transform (Modify) Bone - Bone: dyn_coat_fr_r_MSTR"!(ComponentPose, Alpha=1, Rotation="9.941,-0.024,64.992") -> (Pose) {
    rotationMode=BMM_Replace
    rotationSpace=BCS_BoneSpace
    scaleSpace=BCS_BoneSpace
    alphaBoolBlend {blendInTime=0.7, blendOption=Linear, blendOutTime=0.5}
    alphaScaleBiasClamp {interpSpeedDecreasing=5, interpSpeedIncreasing=7}
    boneToModify {boneName=dyn_coat_fr_r_MSTR}
    rotation=<rot 9.941 -0.0241 64.992>
    scale=<0.8 0.8 0.8>
   }
   n30 = "Transform (Modify) Bone - Bone: dyn_coat_bk_r_MSTR"!(ComponentPose, Alpha=1, Rotation="10,-0.000,0.000") -> (Pose) {
    rotationMode=BMM_Additive
    rotationSpace=BCS_BoneSpace
    alphaBoolBlend {blendOption=Linear}
    boneToModify {boneName=dyn_coat_bk_r_MSTR}
    rotation {pitch=10}
    scale=<0.1 0.1 0.1>
   }
   n31 = "Transform (Modify) Bone - Bone: dyn_coat_bk_MSTR"!(ComponentPose=@n30, Alpha=1, Rotation="15,-0.000,0.000") -> (Pose) {
    rotationMode=BMM_Additive
    rotationSpace=BCS_BoneSpace
    alphaBoolBlend {blendOption=Linear}
    boneToModify {boneName=dyn_coat_bk_MSTR}
    rotation {pitch=15}
    scale=<0.1 0.1 0.1>
   }
   "Transform (Modify) Bone - Bone: dyn_coat_bk_l_MSTR"!(ComponentPose=@n31, Alpha=1, Rotation="10,0.000,-0.000") -> (Pose) {
    rotationMode=BMM_Additive
    rotationSpace=BCS_BoneSpace
    alphaBoolBlend {blendOption=Linear}
    boneToModify {boneName=dyn_coat_bk_l_MSTR}
    rotation {pitch=10}
    scale=<0.1 0.1 0.1>
   }
   n32 = "Transform (Modify) Bone - Bone: dyn_coat_bk_l_MSTR"(ComponentPose=Local To Component(LocalPose=Use cached pose 'RBAN_Off_Base'()), Alpha=1, Rotation="10,0,0") {
    rotationMode=BMM_Replace
    rotationSpace=BCS_BoneSpace
    scaleSpace=BCS_BoneSpace
    alphaBoolBlend {blendInTime=0.7, blendOption=Linear, blendOutTime=0.5}
    alphaScaleBiasClamp {interpSpeedDecreasing=5, interpSpeedIncreasing=7}
    boneToModify {boneName=dyn_coat_bk_l_MSTR}
    rotation {pitch=10}
    scale=<0.8 0.8 0.8>
   }
   n33 = "Transform (Modify) Bone - Bone: dyn_coat_bk_l_04"!(ComponentPose=@n32, Alpha=1, Rotation="10,0,0") -> (Pose) {
    rotationMode=BMM_Replace
    rotationSpace=BCS_BoneSpace
    scaleSpace=BCS_BoneSpace
    alphaBoolBlend {blendInTime=0.7, blendOption=Linear, blendOutTime=0.5}
    alphaScaleBiasClamp {interpSpeedDecreasing=5, interpSpeedIncreasing=7}
    boneToModify {boneName=dyn_coat_bk_l_04}
    rotation {pitch=10}
    scale=<0.8 0.8 0.8>
   }
   n34 = "Transform (Modify) Bone - Bone: dyn_coat_bk_l_05"!(ComponentPose=@n33, Alpha=1, Rotation="10,0,0") -> (Pose) {
    rotationMode=BMM_Replace
    rotationSpace=BCS_BoneSpace
    scaleSpace=BCS_BoneSpace
    alphaBoolBlend {blendInTime=0.7, blendOption=Linear, blendOutTime=0.5}
    alphaScaleBiasClamp {interpSpeedDecreasing=5, interpSpeedIncreasing=7}
    boneToModify {boneName=dyn_coat_bk_l_05}
    rotation {pitch=10}
    scale=<0.8 0.8 0.8>
   }
   n35 = "Transform (Modify) Bone - Bone: dyn_coat_fr_l_MSTR"!(ComponentPose=@n34, Alpha=1, Rotation="5,0,0") -> (Pose) {
    rotationMode=BMM_Replace
    rotationSpace=BCS_BoneSpace
    scaleSpace=BCS_BoneSpace
    alphaBoolBlend {blendInTime=0.7, blendOption=Linear, blendOutTime=0.5}
    alphaScaleBiasClamp {interpSpeedDecreasing=5, interpSpeedIncreasing=7}
    boneToModify {boneName=dyn_coat_fr_l_MSTR}
    rotation {pitch=5}
    scale=<0.8 0.8 0.8>
   }
   n36 = "Transform (Modify) Bone - Bone: dyn_coat_l_MSTR"!(ComponentPose=@n35, Alpha=1, Rotation="5,0,0") -> (Pose) {
    rotationMode=BMM_Replace
    rotationSpace=BCS_BoneSpace
    scaleSpace=BCS_BoneSpace
    alphaBoolBlend {blendInTime=0.7, blendOption=Linear, blendOutTime=0.5}
    alphaScaleBiasClamp {interpSpeedDecreasing=5, interpSpeedIncreasing=7}
    boneToModify {boneName=dyn_coat_l_MSTR}
    rotation {pitch=5}
    scale=<0.8 0.8 0.8>
   }
   n37 = "Transform (Modify) Bone - Bone: dyn_coat_fr_l_03"!(ComponentPose=@n36, Alpha=1, Rotation="10,0,0") -> (Pose) {
    rotationMode=BMM_Replace
    rotationSpace=BCS_BoneSpace
    scaleSpace=BCS_BoneSpace
    alphaBoolBlend {blendInTime=0.7, blendOption=Linear, blendOutTime=0.5}
    alphaScaleBiasClamp {interpSpeedDecreasing=5, interpSpeedIncreasing=7}
    boneToModify {boneName=dyn_coat_fr_l_03}
    rotation {pitch=10}
    scale=<0.8 0.8 0.8>
   }
   n38 = "Transform (Modify) Bone - Bone: dyn_coat_l_03"!(ComponentPose=@n37, Alpha=1, Rotation="10,0,0") -> (Pose) {
    rotationMode=BMM_Replace
    rotationSpace=BCS_BoneSpace
    scaleSpace=BCS_BoneSpace
    alphaBoolBlend {blendInTime=0.7, blendOption=Linear, blendOutTime=0.5}
    alphaScaleBiasClamp {interpSpeedDecreasing=5, interpSpeedIncreasing=7}
    boneToModify {boneName=dyn_coat_l_03}
    rotation {pitch=10}
    scale=<0.8 0.8 0.8>
   }
   n39 = "Transform (Modify) Bone - Bone: dyn_coat_fr_l_05"!(ComponentPose=@n38, Alpha=1, Rotation="10,0,0") -> (Pose) {
    rotationMode=BMM_Replace
    rotationSpace=BCS_BoneSpace
    scaleSpace=BCS_BoneSpace
    alphaBoolBlend {blendInTime=0.7, blendOption=Linear, blendOutTime=0.5}
    alphaScaleBiasClamp {interpSpeedDecreasing=5, interpSpeedIncreasing=7}
    boneToModify {boneName=dyn_coat_fr_l_05}
    rotation {pitch=10}
    scale=<0.8 0.8 0.8>
   }
   n40 = "Transform (Modify) Bone - Bone: dyn_coat_l_05"!(ComponentPose=@n39, Alpha=1, Rotation="10,0,0") -> (Pose) {
    rotationMode=BMM_Replace
    rotationSpace=BCS_BoneSpace
    scaleSpace=BCS_BoneSpace
    alphaBoolBlend {blendInTime=0.7, blendOption=Linear, blendOutTime=0.5}
    alphaScaleBiasClamp {interpSpeedDecreasing=5, interpSpeedIncreasing=7}
    boneToModify {boneName=dyn_coat_l_05}
    rotation {pitch=10}
    scale=<0.8 0.8 0.8>
   }
   n41 = "Transform (Modify) Bone - Bone: dyn_coat_fr_l_06"!(ComponentPose=@n40, Alpha=1, Rotation="25,0,0") -> (Pose) {
    rotationMode=BMM_Replace
    rotationSpace=BCS_BoneSpace
    scaleSpace=BCS_BoneSpace
    alphaBoolBlend {blendInTime=0.7, blendOption=Linear, blendOutTime=0.5}
    alphaScaleBiasClamp {interpSpeedDecreasing=5, interpSpeedIncreasing=7}
    boneToModify {boneName=dyn_coat_fr_l_06}
    rotation {pitch=25}
    scale=<0.8 0.8 0.8>
   }
   n42 = "Transform (Modify) Bone - Bone: dyn_coat_bk_03"!(ComponentPose=@n41, Alpha=1, Rotation="35,0,0") -> (Pose) {
    rotationMode=BMM_Replace
    rotationSpace=BCS_BoneSpace
    alphaBoolBlend {blendInTime=0.7, blendOption=Linear, blendOutTime=0.5}
    alphaScaleBiasClamp {interpSpeedDecreasing=5, interpSpeedIncreasing=7}
    boneToModify {boneName=dyn_coat_bk_03}
    rotation {pitch=35}
    scale=<0.1 0.1 0.1>
   }
   n43 = "Transform (Modify) Bone - Bone: dyn_coat_bkl_05"!(ComponentPose=@n42, Alpha=1, Rotation="12,0,0") -> (Pose) {
    rotationMode=BMM_Replace
    rotationSpace=BCS_BoneSpace
    alphaBoolBlend {blendInTime=0.7, blendOption=Linear, blendOutTime=0.5}
    alphaScaleBiasClamp {interpSpeedDecreasing=5, interpSpeedIncreasing=7}
    boneToModify {boneName=dyn_coat_bkl_05}
    rotation {pitch=12}
    scale=<0.1 0.1 0.1>
   }
   n44 = "Transform (Modify) Bone - Bone: dyn_coat_r_MSTR"!(ComponentPose=@n43, Alpha=1, Rotation="7,0,0") -> (Pose) {
    rotationMode=BMM_Replace
    rotationSpace=BCS_BoneSpace
    alphaBoolBlend {blendInTime=0.7, blendOption=Linear, blendOutTime=0.5}
    alphaScaleBiasClamp {interpSpeedDecreasing=5, interpSpeedIncreasing=7}
    boneToModify {boneName=dyn_coat_r_MSTR}
    rotation {pitch=7}
    scale=<0.1 0.1 0.1>
   }
   n45 = "Transform (Modify) Bone - Bone: dyn_coat_fr_r_MSTR"(ComponentPose=@n44, Alpha=1, Rotation="10,0,0") {
    rotationMode=BMM_Replace
    rotationSpace=BCS_BoneSpace
    alphaBoolBlend {blendInTime=0.7, blendOption=Linear, blendOutTime=0.5}
    alphaScaleBiasClamp {interpSpeedDecreasing=5, interpSpeedIncreasing=7}
    boneToModify {boneName=dyn_coat_fr_r_MSTR}
    rotation {pitch=10}
    scale=<0.1 0.1 0.1>
   }
   n46 = "Transform (Modify) Bone - Bone: dyn_coat_r_03"!(ComponentPose=@n45, Alpha=1, Rotation="10,0,0") -> (Pose) {
    rotationMode=BMM_Replace
    rotationSpace=BCS_BoneSpace
    alphaBoolBlend {blendInTime=0.7, blendOption=Linear, blendOutTime=0.5}
    alphaScaleBiasClamp {interpSpeedDecreasing=5, interpSpeedIncreasing=7}
    boneToModify {boneName=dyn_coat_r_03}
    rotation {pitch=10}
    scale=<0.1 0.1 0.1>
   }
   n47 = "Transform (Modify) Bone - Bone: dyn_coat_fr_r_03"!(ComponentPose=@n46, Alpha=1, Rotation="10,0,0") -> (Pose) {
    rotationMode=BMM_Replace
    rotationSpace=BCS_BoneSpace
    alphaBoolBlend {blendInTime=0.7, blendOption=Linear, blendOutTime=0.5}
    alphaScaleBiasClamp {interpSpeedDecreasing=5, interpSpeedIncreasing=7}
    boneToModify {boneName=dyn_coat_fr_r_03}
    rotation {pitch=10}
    scale=<0.1 0.1 0.1>
   }
   n48 = "Transform (Modify) Bone - Bone: dyn_coat_r_05"!(ComponentPose=@n47, Alpha=1, Rotation="25,0,0") -> (Pose) {
    rotationMode=BMM_Replace
    rotationSpace=BCS_BoneSpace
    alphaBoolBlend {blendInTime=0.7, blendOption=Linear, blendOutTime=0.5}
    alphaScaleBiasClamp {interpSpeedDecreasing=5, interpSpeedIncreasing=7}
    boneToModify {boneName=dyn_coat_r_05}
    rotation {pitch=25}
    scale=<0.1 0.1 0.1>
   }
   n49 = "Transform (Modify) Bone - Bone: dyn_coat_fr_r_05"!(ComponentPose=@n48, Alpha=1, Rotation="25,0,0") -> (Pose) {
    rotationMode=BMM_Replace
    rotationSpace=BCS_BoneSpace
    alphaBoolBlend {blendInTime=0.7, blendOption=Linear, blendOutTime=0.5}
    alphaScaleBiasClamp {interpSpeedDecreasing=5, interpSpeedIncreasing=7}
    boneToModify {boneName=dyn_coat_fr_r_05}
    rotation {pitch=25}
    scale=<0.1 0.1 0.1>
   }
   n50 = "Transform (Modify) Bone - Bone: dyn_coat_r_06"!(ComponentPose=@n49, Alpha=1, Rotation="20,0,0") -> (Pose) {
    rotationMode=BMM_Replace
    rotationSpace=BCS_BoneSpace
    alphaBoolBlend {blendInTime=0.7, blendOption=Linear, blendOutTime=0.5}
    alphaScaleBiasClamp {interpSpeedDecreasing=5, interpSpeedIncreasing=7}
    boneToModify {boneName=dyn_coat_r_06}
    rotation {pitch=20}
    scale=<0.1 0.1 0.1>
   }
   n51 = "Transform (Modify) Bone - Bone: dyn_coat_fr_r_06"!(ComponentPose=@n50, Alpha=1, Rotation="20,0,0") -> (Pose) {
    rotationMode=BMM_Replace
    rotationSpace=BCS_BoneSpace
    alphaBoolBlend {blendInTime=0.7, blendOption=Linear, blendOutTime=0.5}
    alphaScaleBiasClamp {interpSpeedDecreasing=5, interpSpeedIncreasing=7}
    boneToModify {boneName=dyn_coat_fr_r_06}
    rotation {pitch=20}
    scale=<0.1 0.1 0.1>
   }
   n52 = "Transform (Modify) Bone - Bone: dyn_coat_bk_r_MSTR"(ComponentPose=@n51, Alpha=1, Rotation="15,10,0") {
    rotationMode=BMM_Replace
    rotationSpace=BCS_BoneSpace
    scaleSpace=BCS_BoneSpace
    alphaBoolBlend {blendInTime=0.7, blendOption=Linear, blendOutTime=0.5}
    alphaScaleBiasClamp {interpSpeedDecreasing=5, interpSpeedIncreasing=7}
    boneToModify {boneName=dyn_coat_bk_r_MSTR}
    rotation {pitch=15, yaw=10}
    scale=<0.8 0.8 0.8>
   }
   n53 = "Transform (Modify) Bone - Bone: dyn_coat_bk_r_02"!(ComponentPose=@n52, Alpha=1, Rotation="15,0,0") -> (Pose) {
    rotationMode=BMM_Replace
    rotationSpace=BCS_BoneSpace
    scaleSpace=BCS_BoneSpace
    alphaBoolBlend {blendInTime=0.7, blendOption=Linear, blendOutTime=0.5}
    alphaScaleBiasClamp {interpSpeedDecreasing=5, interpSpeedIncreasing=7}
    boneToModify {boneName=dyn_coat_bk_r_02}
    rotation {pitch=15}
    scale=<0.8 0.8 0.8>
   }
   n54 = "Transform (Modify) Bone - Bone: dyn_coat_bk_r_03"!(ComponentPose=@n53, Alpha=1, Rotation="18,0,0") -> (Pose) {
    rotationMode=BMM_Replace
    rotationSpace=BCS_BoneSpace
    scaleSpace=BCS_BoneSpace
    alphaBoolBlend {blendInTime=0.7, blendOption=Linear, blendOutTime=0.5}
    alphaScaleBiasClamp {interpSpeedDecreasing=5, interpSpeedIncreasing=7}
    boneToModify {boneName=dyn_coat_bk_r_03}
    rotation {pitch=18}
    scale=<0.8 0.8 0.8>
   }
   n55 = "Transform (Modify) Bone - Bone: dyn_coat_bk_r_04"!(ComponentPose=@n54, Alpha=1, Rotation="0,0,0") -> (Pose) {
    rotationMode=BMM_Replace
    rotationSpace=BCS_BoneSpace
    scaleSpace=BCS_BoneSpace
    alphaBoolBlend {blendInTime=0.7, blendOption=Linear, blendOutTime=0.5}
    alphaScaleBiasClamp {interpSpeedDecreasing=5, interpSpeedIncreasing=7}
    boneToModify {boneName=dyn_coat_bk_r_04}
    scale=<0.8 0.8 0.8>
   }
   n56 = "Transform (Modify) Bone - Bone: dyn_coat_bk_r_05"!(ComponentPose=@n55, Alpha=1, Rotation="0,0,0") -> (Pose) {
    rotationMode=BMM_Replace
    rotationSpace=BCS_BoneSpace
    scaleSpace=BCS_BoneSpace
    alphaBoolBlend {blendInTime=0.7, blendOption=Linear, blendOutTime=0.5}
    alphaScaleBiasClamp {interpSpeedDecreasing=5, interpSpeedIncreasing=7}
    boneToModify {boneName=dyn_coat_bk_r_05}
    scale=<0.8 0.8 0.8>
   }
   n57 = "Transform (Modify) Bone - Bone: dyn_coat_r_MSTR"(ComponentPose=@n56, Alpha=1, Rotation="10.016,-0.030,-5.008") {
    rotationMode=BMM_Replace
    rotationSpace=BCS_BoneSpace
    scaleSpace=BCS_BoneSpace
    alphaBoolBlend {blendInTime=0.7, blendOption=Linear, blendOutTime=0.5}
    alphaScaleBiasClamp {interpSpeedDecreasing=5, interpSpeedIncreasing=7}
    boneToModify {boneName=dyn_coat_r_MSTR}
    rotation=<rot 10.016 -0.03 -5.008>
    scale=<0.8 0.8 0.8>
   }
   n58 = "Transform (Modify) Bone - Bone: dyn_coat_l_03"(ComponentPose=@n57, Alpha=1, Rotation="15,0,0") {
    rotationMode=BMM_Replace
    rotationSpace=BCS_BoneSpace
    scaleSpace=BCS_BoneSpace
    alphaBoolBlend {blendInTime=0.7, blendOption=Linear, blendOutTime=0.5}
    alphaScaleBiasClamp {interpSpeedDecreasing=5, interpSpeedIncreasing=7}
    boneToModify {boneName=dyn_coat_l_03}
    rotation {pitch=15}
    scale=<0.8 0.8 0.8>
   }
   n59 = "Transform (Modify) Bone - Bone: dyn_coat_r_03"(ComponentPose=@n58, Alpha=1, Rotation="15,0,0") {
    rotationMode=BMM_Replace
    rotationSpace=BCS_BoneSpace
    alphaBoolBlend {blendInTime=0.7, blendOption=Linear, blendOutTime=0.5}
    alphaScaleBiasClamp {interpSpeedDecreasing=5, interpSpeedIncreasing=7}
    boneToModify {boneName=dyn_coat_r_03}
    rotation {pitch=15}
    scale=<0.1 0.1 0.1>
   }
   n60 = Get thigh_r_alpha() -> (thigh_r_alpha:double)
   n61 = "Transform (Modify) Bone - Bone: dyn_coat_fr_r_01"(ComponentPose=@n59, Alpha=@n60, Translation="0,0,2", Rotation="20,0,0", Scale="1,1,1") {
    rotationMode=BMM_Additive
    rotationSpace=BCS_BoneSpace
    translationMode=BMM_Additive
    translationSpace=BCS_BoneSpace
    alphaBoolBlend {blendOption=Linear}
    boneToModify {boneName=dyn_coat_fr_r_01}
    rotation {pitch=20}
    translation {z=2}
   }
   n62 = "Transform (Modify) Bone - Bone: dyn_coat_fr_r_01"(ComponentPose=@n61, Alpha=@n60, Translation="0,0,1", Rotation="10,0,0", Scale="1,1,1") {
    rotationMode=BMM_Additive
    rotationSpace=BCS_BoneSpace
    translationMode=BMM_Additive
    translationSpace=BCS_BoneSpace
    alphaBoolBlend {blendOption=Linear}
    boneToModify {boneName=dyn_coat_fr_r_01}
    rotation {pitch=10}
    translation {z=1}
   }
   n63 = Blend Poses by bool(
    BlendPose_0=Component To Local(ComponentPose=@n62)
    BlendPose_1=Use cached pose 'RBAN_Off_Base'()
    BlendTime_0=0.08
    BlendTime_1=0.3
    bActiveValue=Get bIsCrouching()
   ) {blendType=Linear}
   n64 = RigidBody!(
    ComponentPose=Local To Component(LocalPose=Use cached pose 'main_SavedPose'())
    Alpha=Get in vehicle rigid alpha() -> (in vehicle rigid alpha:double)
    OverrideWorldGravity:Vector=Get gravity_override() -> (gravity_override:Vector)
    ExternalForce
   ) -> (Pose) {
    bClampLinearTranslationLimitToRefPose=true
    bOverrideWorldGravity=true
    overridePhysicsAsset=&/M_MED_SkeletonHunterGlitch_Physics
    simulationSpace=BaseBoneSpace
    alphaBoolBlend {blendOption=Linear}
    baseBoneRef {boneName=pelvis}
    componentAppliedLinearAccClamp=<6000 6000 6000>
    componentLinearAccScale=<0.5 0.4 0.4>
    componentLinearVelScale=<0.5 0.4 0.4>
    overrideWorldGravity {z=-980}
   }
   n65 = Component To Local(ComponentPose=@n64)
   n66 = RigidBody!(ComponentPose=Local To Component(LocalPose=Use cached pose 'main_SavedPose'()), Alpha, ExternalForce) -> (Pose) {
    bClampLinearTranslationLimitToRefPose=true
    bOverrideWorldGravity=true
    overridePhysicsAsset=&/M_MED_SkeletonHunterGlitch_Physics_Crouching
    simulationSpace=BaseBoneSpace
    alphaBoolBlend {blendOption=Linear}
    baseBoneRef {boneName=pelvis}
    componentAppliedLinearAccClamp=<4000 4000 4000>
    componentLinearAccScale=<0.3 0.3 0.3>
    componentLinearVelScale=<0.3 0.3 0.3>
    overrideWorldGravity {z=-200}
   }
   n67 = "Transform (Modify) Bone - Bone: dyn_coat_bk_l_MSTR"(ComponentPose=@n66, Alpha=1, Rotation="10,0,0") {
    rotationMode=BMM_Replace
    rotationSpace=BCS_BoneSpace
    scaleSpace=BCS_BoneSpace
    alphaBoolBlend {blendInTime=0.7, blendOption=Linear, blendOutTime=0.5}
    alphaScaleBiasClamp {interpSpeedDecreasing=5, interpSpeedIncreasing=7}
    boneToModify {boneName=dyn_coat_bk_l_MSTR}
    rotation {pitch=10}
    scale=<0.8 0.8 0.8>
   }
   n68 = "Transform (Modify) Bone - Bone: dyn_coat_bk_l_04"(ComponentPose=@n67, Alpha=1, Rotation="10,0,0") {
    rotationMode=BMM_Replace
    rotationSpace=BCS_BoneSpace
    scaleSpace=BCS_BoneSpace
    alphaBoolBlend {blendInTime=0.7, blendOption=Linear, blendOutTime=0.5}
    alphaScaleBiasClamp {interpSpeedDecreasing=5, interpSpeedIncreasing=7}
    boneToModify {boneName=dyn_coat_bk_l_04}
    rotation {pitch=10}
    scale=<0.8 0.8 0.8>
   }
   n69 = "Transform (Modify) Bone - Bone: dyn_coat_bk_r_MSTR"(ComponentPose=@n68, Alpha=1, Rotation="15,10,0") {
    rotationMode=BMM_Replace
    rotationSpace=BCS_BoneSpace
    scaleSpace=BCS_BoneSpace
    alphaBoolBlend {blendInTime=0.7, blendOption=Linear, blendOutTime=0.5}
    alphaScaleBiasClamp {interpSpeedDecreasing=5, interpSpeedIncreasing=7}
    boneToModify {boneName=dyn_coat_bk_r_MSTR}
    rotation {pitch=15, yaw=10}
    scale=<0.8 0.8 0.8>
   }
   n70 = "Transform (Modify) Bone - Bone: dyn_coat_r_MSTR"(ComponentPose=@n69, Alpha=1, Rotation="10.016,-0.030,-5.008") {
    rotationMode=BMM_Replace
    rotationSpace=BCS_BoneSpace
    scaleSpace=BCS_BoneSpace
    alphaBoolBlend {blendInTime=0.7, blendOption=Linear, blendOutTime=0.5}
    alphaScaleBiasClamp {interpSpeedDecreasing=5, interpSpeedIncreasing=7}
    boneToModify {boneName=dyn_coat_r_MSTR}
    rotation=<rot 10.016 -0.03 -5.008>
    scale=<0.8 0.8 0.8>
   }
   n71 = Look At - Bone: dyn_coat_bk_03(ComponentPose=@n70, Alpha=1) {
    alphaBoolBlend {blendOption=Linear}
    alphaScaleBiasClamp {bClampResult=true, bInterpResult=true, clampMax=0.8, interpSpeedDecreasing=5, interpSpeedIncreasing=7}
    boneToModify {boneName=dyn_coat_bk_03}
    lookAtLocation=<78.4907 -6.6912 -13.1281>
    lookAtTarget {boneReference {boneName=calf_r}}
    lookAt_Axis {axis=<1 0>}
   }
   n72 = Look At - Bone: dyn_coat_bk_r_03(ComponentPose=@n71, Alpha=1) {
    alphaBoolBlend {blendOption=Linear}
    alphaScaleBiasClamp {bClampResult=true, bInterpResult=true, clampMax=0.8, interpSpeedDecreasing=5, interpSpeedIncreasing=7}
    boneToModify {boneName=dyn_coat_bk_r_03}
    lookAtLocation=<78.9142 -24.2868 9.6539>
    lookAtTarget {boneReference {boneName=calf_r}}
    lookAt_Axis {axis=<1 0>}
   }
   n73 = Component To Local(ComponentPose=@n72)
   n74 = Get bIsCrouching()
   n75 = RigidBody(ComponentPose=Local To Component(LocalPose=Use cached pose 'main_SavedPose'())) {
    bClampLinearTranslationLimitToRefPose=true
    bOverrideWorldGravity=true
    overridePhysicsAsset=&/M_MED_SkeletonHunterGlitch_Physics_Skydive
    simulationSpace=BaseBoneSpace
    alphaBoolBlend {blendOption=Linear}
    baseBoneRef {boneName=pelvis}
    componentAppliedLinearAccClamp=<3000 3000 3000>
    componentLinearAccScale=<0.3 0.15 0.15>
    componentLinearVelScale=<0.3 0.15 0.15>
    overrideWorldGravity {z=-200}
   }
   n76 = RigidBody(ComponentPose=Local To Component(LocalPose=Use cached pose 'main_SavedPose'())) {
    bClampLinearTranslationLimitToRefPose=true
    bOverrideWorldGravity=true
    overridePhysicsAsset=&M_MED_SkeletonHunterGlitch_Physics
    simulationSpace=BaseBoneSpace
    alphaBoolBlend {blendOption=Linear}
    baseBoneRef {boneName=root}
    componentAppliedLinearAccClamp=<3000 3000 3000>
    componentLinearAccScale=<1.5 1.5 1.5>
    componentLinearVelScale=<1.5 1.5 1.5>
    overrideWorldGravity {z=-300}
   }
   n77 = Blend Poses by bool(
    BlendPose_0=Component To Local(ComponentPose=@n76)
    BlendPose_1=Blend Poses by bool(
     BlendPose_0=Component To Local(ComponentPose=@n75)
     BlendPose_1=Blend Poses by bool(BlendPose_0=@n73, BlendPose_1=@n65, BlendTime_0=0.7, BlendTime_1=0.5, bActiveValue=@n74) {blendType=Linear}
     bActiveValue=Get bIsSkyDiving()
    ) {blendType=Linear}
    bActiveValue=Get bIsPlayingEmote()
   ) {blendType=Linear}
   n78 = Get bRigidBodyEnabled()
   n79 = "Transform (Modify) Bone - Bone: dyn_coat_fr_r_03"(
    ComponentPose=Local To Component(LocalPose=Blend Poses by bool(BlendPose_0=@n77, BlendPose_1=@n63, bActiveValue=@n78) {blendType=Linear})
    Alpha=1
    Rotation="-23.704,23.491,11.949"
   ) {
    rotationMode=BMM_Replace
    rotationSpace=BCS_BoneSpace
    scaleSpace=BCS_BoneSpace
    alphaBoolBlend {blendInTime=0.7, blendOption=Linear, blendOutTime=0.5}
    alphaScaleBiasClamp {bInterpResult=true, interpSpeedDecreasing=5, interpSpeedIncreasing=7}
    boneToModify {boneName=dyn_coat_fr_r_03}
    rotation=<rot -23.704 23.4909 11.949>
    scale=<0.8 0.8 0.8>
   }
   n80 = "Transform (Modify) Bone - Bone: dyn_coat_fr_r_04"!(ComponentPose=@n79, Alpha=1, Rotation="32.690,-19.361,-5.588") -> (Pose) {
    rotationMode=BMM_Replace
    rotationSpace=BCS_BoneSpace
    scaleSpace=BCS_BoneSpace
    alphaBoolBlend {blendInTime=0.7, blendOption=Linear, blendOutTime=0.5}
    alphaScaleBiasClamp {bInterpResult=true, interpSpeedDecreasing=5, interpSpeedIncreasing=7}
    boneToModify {boneName=dyn_coat_fr_r_04}
    rotation=<rot 32.69 -19.361 -5.588>
    scale=<0.8 0.8 0.8>
   }
   n81 = "Transform (Modify) Bone - Bone: dyn_coat_r_05"(ComponentPose=@n80, Alpha=1, Rotation="-9.775,10.004,0.002") {
    rotationMode=BMM_Replace
    rotationSpace=BCS_BoneSpace
    scaleSpace=BCS_BoneSpace
    alphaBoolBlend {blendInTime=0.7, blendOption=Linear, blendOutTime=0.5}
    alphaScaleBiasClamp {bInterpResult=true, interpSpeedDecreasing=5, interpSpeedIncreasing=7}
    boneToModify {boneName=dyn_coat_r_05}
    rotation=<rot -9.775 10.004 0.002>
    scale=<0.8 0.8 0.8>
   }
   n82 = "Transform (Modify) Bone - Bone: dyn_coat_bk_r_05"(ComponentPose=@n81, Alpha=1, Rotation="-19.690,10.645,-3.571") {
    rotationMode=BMM_Replace
    rotationSpace=BCS_BoneSpace
    scaleSpace=BCS_BoneSpace
    alphaBoolBlend {blendInTime=0.7, blendOption=Linear, blendOutTime=0.5}
    alphaScaleBiasClamp {bInterpResult=true, interpSpeedDecreasing=5, interpSpeedIncreasing=7}
    boneToModify {boneName=dyn_coat_bk_r_05}
    rotation=<rot -19.69 10.645 -3.571>
    scale=<0.8 0.8 0.8>
   }
   Output Pose(Result=Component To Local(ComponentPose=@n82)) {layerGroup=None}
   n83 = Get in vehicle scale alpha() -> (in vehicle scale alpha:double)
   n84 = "Transform (Modify) Bone - Bone: dyn_coat_l_02"(
    ComponentPose=Local To Component(LocalPose=Copy Pose From Mesh() {bUseAttachedParent=true})
    Alpha=@n83
    Translation="0,0,0"
    Rotation="0,0,0"
    Scale="0.1,0.1,0.1"
   ) {scaleMode=BMM_Replace, alphaBoolBlend {blendOption=Linear}, boneToModify {boneName=dyn_coat_l_02}, scale=<0.1 0.1 0.1>}
   n85 = "Transform (Modify) Bone - Bone: dyn_coat_fr_l_02"(ComponentPose=@n84, Alpha=@n83, Translation="0,0,0", Rotation="0,0,0", Scale="0.1,0.1,0.1") {scaleMode=BMM_Replace, alphaBoolBlend {blendOption=Linear}, boneToModify {boneName=dyn_coat_fr_l_02}, scale=<0.1 0.1 0.1>}
   n86 = "Transform (Modify) Bone - Bone: dyn_coat_bk_02"(ComponentPose=@n85, Alpha=@n83, Translation="0,0,0", Rotation="0,0,0", Scale="0.1,0.1,0.1") {scaleMode=BMM_Replace, alphaBoolBlend {blendOption=Linear}, boneToModify {boneName=dyn_coat_bk_02}, scale=<0.1 0.1 0.1>}
   n87 = "Transform (Modify) Bone - Bone: dyn_coat_bk_l_02"(ComponentPose=@n86, Alpha=@n83, Translation="0,0,0", Rotation="0,0,0", Scale="0.1,0.1,0.1") {scaleMode=BMM_Replace, alphaBoolBlend {blendOption=Linear}, boneToModify {boneName=dyn_coat_bk_l_02}, scale=<0.1 0.1 0.1>}
   n88 = "Transform (Modify) Bone - Bone: dyn_coat_bk_r_02"!(ComponentPose=@n87, Alpha=@n83, Translation:Vector="0,0,0", Rotation="0,0,0", Scale:Vector="0.1,0.1,0.1") -> (Pose) {scaleMode=BMM_Replace, alphaBoolBlend {blendOption=Linear}, boneToModify {boneName=dyn_coat_bk_r_02}, scale=<0.1 0.1 0.1>}
   n89 = "Transform (Modify) Bone - Bone: dyn_coat_fr_r_02"(ComponentPose=@n88, Alpha=@n83, Translation="0,0,0", Rotation="0,0,0", Scale="0.1,0.1,0.1") {scaleMode=BMM_Replace, alphaBoolBlend {blendOption=Linear}, boneToModify {boneName=dyn_coat_fr_r_02}, scale=<0.1 0.1 0.1>}
   n90 = "Transform (Modify) Bone - Bone: dyn_coat_r_02"(ComponentPose=@n89, Alpha=@n83, Translation="0,0,0", Rotation="0,0,0", Scale="0.1,0.1,0.1") {scaleMode=BMM_Replace, alphaBoolBlend {blendOption=Linear}, boneToModify {boneName=dyn_coat_r_02}, scale=<0.1 0.1 0.1>}
   n91 = Get Alpha_sliding() -> (Alpha_sliding:double)
   n92 = Get Alpha_jump_right() -> (Alpha_jump_right:double)
   n93 = Control Rig!(
    Source=Component To Local(ComponentPose=@n90)
    Alpha=Get RBAN on() -> (RBAN on:double)
    Alpha_crouching:double=0
    Alpha_jump_right:double=@n92
    Alpha_sliding:double=@n91
   ) -> (Pose) {
    alphaBoolBlend {blendOption=Linear}
    controlRigAssetReference {blueprintRigClass=&/M_MED_SkeletonHunterGlitch_CtrlRig.M_MED_SkeletonHunterGlitch_CtrlRig_C}
   }
   n94 = Control Rig(Source=@n93, Alpha=Get RBAN off() -> (RBAN off:double), Alpha_crouching=0, Alpha_jump_right=@n92, Alpha_sliding=@n91) {
    alphaBoolBlend {blendOption=Linear}
    controlRigAssetReference {blueprintRigClass=&/M_MED_SkeletonHunterGlitch_RBANOff_CtrlRig.M_MED_SkeletonHunterGlitch_RBANOff_CtrlRig_C}
   }
   Save cached pose 'main_SavedPose'(Pose=@n94) {cacheName=main_SavedPose}
   n95 = "Constraint - dyn_coat_fr_l_02 (PP)"(ComponentPose=Local To Component(LocalPose=Use cached pose 'main_SavedPose'()), Alpha=1, ConstraintWeights_0=0.2, ConstraintWeights_1=0.8) {
    alphaBoolBlend {blendOption=Linear}
    boneToModify {boneName=dyn_coat_fr_l_02}
    constraintSetup [{targetBone {boneName=thigh_l}}, {targetBone {boneName=pelvis}}]
    constraintWeights [1,1]
   }
   n96 = "Constraint - dyn_coat_l_02 (PP)"(ComponentPose=@n95, Alpha=1, ConstraintWeights_0=0.2, ConstraintWeights_1=0.8) {
    alphaBoolBlend {blendOption=Linear}
    boneToModify {boneName=dyn_coat_l_02}
    constraintSetup [{targetBone {boneName=thigh_l}}, {targetBone {boneName=pelvis}}]
    constraintWeights [1,1]
   }
   n97 = "Constraint - dyn_coat_bk_l_02 (PP)"(ComponentPose=@n96, Alpha=1, ConstraintWeights_0=0.2, ConstraintWeights_1=0.8) {
    alphaBoolBlend {blendOption=Linear}
    boneToModify {boneName=dyn_coat_bk_l_02}
    constraintSetup [{targetBone {boneName=thigh_l}}, {targetBone {boneName=pelvis}}]
    constraintWeights [1,1]
   }
   n98 = "Constraint - dyn_coat_fr_l_03 (PP)"(ComponentPose=@n97, Alpha=1, ConstraintWeights_0=0.5, ConstraintWeights_1=0.5) {
    alphaBoolBlend {blendOption=Linear}
    boneToModify {boneName=dyn_coat_fr_l_03}
    constraintSetup [{targetBone {boneName=thigh_l}}, {targetBone {boneName=pelvis}}]
    constraintWeights [1,1]
   }
   n99 = "Constraint - dyn_coat_l_03 (PP)"(ComponentPose=@n98, Alpha=1, ConstraintWeights_0=0.5, ConstraintWeights_1=0.5) {
    alphaBoolBlend {blendOption=Linear}
    boneToModify {boneName=dyn_coat_l_03}
    constraintSetup [{targetBone {boneName=thigh_l}}, {targetBone {boneName=pelvis}}]
    constraintWeights [1,1]
   }
   n100 = "Constraint - dyn_coat_bk_l_03 (PP)"(ComponentPose=@n99, Alpha=1, ConstraintWeights_0=0.5, ConstraintWeights_1=0.5) {
    alphaBoolBlend {blendOption=Linear}
    boneToModify {boneName=dyn_coat_bk_l_03}
    constraintSetup [{targetBone {boneName=thigh_l}}, {targetBone {boneName=pelvis}}]
    constraintWeights [1,1]
   }
   n101 = "Constraint - dyn_coat_fr_l_05 (PP)"(ComponentPose=@n100, Alpha=1, ConstraintWeights_0=0.5, ConstraintWeights_1=0.5) {
    alphaBoolBlend {blendOption=Linear}
    boneToModify {boneName=dyn_coat_fr_l_05}
    constraintSetup [{targetBone {boneName=thigh_l}}, {targetBone {boneName=calf_l}}]
    constraintWeights [1,1]
   }
   n102 = "Constraint - dyn_coat_l_05 (PP)"(ComponentPose=@n101, Alpha=1, ConstraintWeights_0=0.5, ConstraintWeights_1=0.5) {
    alphaBoolBlend {blendOption=Linear}
    boneToModify {boneName=dyn_coat_l_05}
    constraintSetup [{targetBone {boneName=thigh_l}}, {targetBone {boneName=calf_l}}]
    constraintWeights [1,1]
   }
   n103 = "Constraint - dyn_coat_bk_l_05 (PP)"(ComponentPose=@n102, Alpha=1, ConstraintWeights_0=0.5, ConstraintWeights_1=0.5) {
    alphaBoolBlend {blendOption=Linear}
    boneToModify {boneName=dyn_coat_bk_l_05}
    constraintSetup [{targetBone {boneName=thigh_l}}, {targetBone {boneName=calf_l}}]
    constraintWeights [1,1]
   }
   n104 = "Constraint - dyn_coat_bk_l_06 (PP)"(ComponentPose=@n103, Alpha=1, ConstraintWeights_0=0.25, ConstraintWeights_1=0.75) {
    alphaBoolBlend {blendOption=Linear}
    boneToModify {boneName=dyn_coat_bk_l_06}
    constraintSetup [{targetBone {boneName=thigh_l}}, {targetBone {boneName=calf_l}}]
    constraintWeights [1,1]
   }
   n105 = "Constraint - dyn_coat_fr_r_02 (PP)"(ComponentPose=@n104, Alpha=1, ConstraintWeights_0=0.2, ConstraintWeights_1=0.8) {
    alphaBoolBlend {blendOption=Linear}
    boneToModify {boneName=dyn_coat_fr_r_02}
    constraintSetup [{targetBone {boneName=thigh_r}}, {targetBone {boneName=pelvis}}]
    constraintWeights [1,1]
   }
   n106 = "Constraint - dyn_coat_r_02 (PP)"(ComponentPose=@n105, Alpha=1, ConstraintWeights_0=0.2, ConstraintWeights_1=0.8) {
    alphaBoolBlend {blendOption=Linear}
    boneToModify {boneName=dyn_coat_r_02}
    constraintSetup [{targetBone {boneName=thigh_r}}, {targetBone {boneName=pelvis}}]
    constraintWeights [1,1]
   }
   n107 = "Constraint - dyn_coat_bk_r_02 (PP)"(ComponentPose=@n106, Alpha=1, ConstraintWeights_0=0.2, ConstraintWeights_1=0.8) {
    alphaBoolBlend {blendOption=Linear}
    boneToModify {boneName=dyn_coat_bk_r_02}
    constraintSetup [{targetBone {boneName=thigh_r}}, {targetBone {boneName=pelvis}}]
    constraintWeights [1,1]
   }
   n108 = "Constraint - dyn_coat_fr_r_03 (PP)"(ComponentPose=@n107, Alpha=1, ConstraintWeights_0=0.5, ConstraintWeights_1=0.5) {
    alphaBoolBlend {blendOption=Linear}
    boneToModify {boneName=dyn_coat_fr_r_03}
    constraintSetup [{targetBone {boneName=thigh_r}}, {targetBone {boneName=pelvis}}]
    constraintWeights [1,1]
   }
   n109 = "Constraint - dyn_coat_r_03 (PP)"(ComponentPose=@n108, Alpha=1, ConstraintWeights_0=0.5, ConstraintWeights_1=0.5) {
    alphaBoolBlend {blendOption=Linear}
    boneToModify {boneName=dyn_coat_r_03}
    constraintSetup [{targetBone {boneName=thigh_r}}, {targetBone {boneName=pelvis}}]
    constraintWeights [1,1]
   }
   n110 = "Constraint - dyn_coat_bk_r_03 (PP)"(ComponentPose=@n109, Alpha=1, ConstraintWeights_0=0.5, ConstraintWeights_1=0.5) {
    alphaBoolBlend {blendOption=Linear}
    boneToModify {boneName=dyn_coat_bk_r_03}
    constraintSetup [{targetBone {boneName=thigh_r}}, {targetBone {boneName=pelvis}}]
    constraintWeights [1,1]
   }
   n111 = "Constraint - dyn_coat_fr_r_05 (PP)"(ComponentPose=@n110, Alpha=1, ConstraintWeights_0=0.5, ConstraintWeights_1=0.5) {
    alphaBoolBlend {blendOption=Linear}
    boneToModify {boneName=dyn_coat_fr_r_05}
    constraintSetup [{targetBone {boneName=thigh_r}}, {targetBone {boneName=calf_r}}]
    constraintWeights [1,1]
   }
   n112 = "Constraint - dyn_coat_r_05 (PP)"(ComponentPose=@n111, Alpha=1, ConstraintWeights_0=0.5, ConstraintWeights_1=0.5) {
    alphaBoolBlend {blendOption=Linear}
    boneToModify {boneName=dyn_coat_r_05}
    constraintSetup [{targetBone {boneName=thigh_r}}, {targetBone {boneName=calf_r}}]
    constraintWeights [1,1]
   }
   n113 = "Constraint - dyn_coat_bk_r_05 (PP)"(ComponentPose=@n112, Alpha=1, ConstraintWeights_0=0.5, ConstraintWeights_1=0.5) {
    alphaBoolBlend {blendOption=Linear}
    boneToModify {boneName=dyn_coat_bk_r_05}
    constraintSetup [{targetBone {boneName=thigh_r}}, {targetBone {boneName=calf_r}}]
    constraintWeights [1,1]
   }
   n114 = "Constraint - dyn_coat_bk_r_06 (PP)"(ComponentPose=@n113, Alpha=1, ConstraintWeights_0=0.25, ConstraintWeights_1=0.75) {
    alphaBoolBlend {blendOption=Linear}
    boneToModify {boneName=dyn_coat_bk_r_06}
    constraintSetup [{targetBone {boneName=thigh_r}}, {targetBone {boneName=calf_r}}]
    constraintWeights [1,1]
   }
   n115 = "Constraint - dyn_coat_bk_02 (PP)"(ComponentPose=@n114, Alpha=1, ConstraintWeights_0=0.5, ConstraintWeights_1=0.5) {
    alphaBoolBlend {blendOption=Linear}
    boneToModify {boneName=dyn_coat_bk_02}
    constraintSetup [{targetBone {boneName=dyn_coat_bk_l_02}}, {targetBone {boneName=dyn_coat_bk_r_02}}]
    constraintWeights [1,1]
   }
   n116 = "Constraint - dyn_coat_bk_03 (PP)"(ComponentPose=@n115, Alpha=1, ConstraintWeights_0=0.5, ConstraintWeights_1=0.5) {
    alphaBoolBlend {blendOption=Linear}
    boneToModify {boneName=dyn_coat_bk_03}
    constraintSetup [{targetBone {boneName=dyn_coat_bk_l_03}}, {targetBone {boneName=dyn_coat_bk_r_03}}]
    constraintWeights [1,1]
   }
   n117 = "Constraint - dyn_coat_bkl_05 (PP)"(ComponentPose=@n116, Alpha=1, ConstraintWeights_0=0.6, ConstraintWeights_1=0.4) {
    alphaBoolBlend {blendOption=Linear}
    boneToModify {boneName=dyn_coat_bkl_05}
    constraintSetup [{targetBone {boneName=dyn_coat_bk_l_05}}, {targetBone {boneName=dyn_coat_bk_r_05}}]
    constraintWeights [1,1]
   }
   n118 = "Constraint - dyn_coat_bkr_05 (PP)"(ComponentPose=@n117, Alpha=1, ConstraintWeights_0=0.4, ConstraintWeights_1=0.6) {
    alphaBoolBlend {blendOption=Linear}
    boneToModify {boneName=dyn_coat_bkr_05}
    constraintSetup [{targetBone {boneName=dyn_coat_bk_l_05}}, {targetBone {boneName=dyn_coat_bk_r_05}}]
    constraintWeights [1,1]
   }
   n119 = "Constraint - dyn_coat_bkl_06 (PP)"(ComponentPose=@n118, Alpha=1, ConstraintWeights_0=0.6, ConstraintWeights_1=0.4) {
    alphaBoolBlend {blendOption=Linear}
    boneToModify {boneName=dyn_coat_bkl_06}
    constraintSetup [{targetBone {boneName=dyn_coat_bk_l_06}}, {targetBone {boneName=dyn_coat_bk_r_06}}]
    constraintWeights [1,1]
   }
   n120 = "Constraint - dyn_coat_bkr_06 (PP)"(ComponentPose=@n119, Alpha=1, ConstraintWeights_0=0.4, ConstraintWeights_1=0.6) {
    alphaBoolBlend {blendOption=Linear}
    boneToModify {boneName=dyn_coat_bkr_06}
    constraintSetup [{targetBone {boneName=dyn_coat_bk_l_06}}, {targetBone {boneName=dyn_coat_bk_r_06}}]
    constraintWeights [1,1]
   }
   Save cached pose 'RBAN_Off_Base'(Pose=Component To Local(ComponentPose=@n120)) {cacheName=RBAN_Off_Base}
   n121 = Get ThighRotInrR() -> (ThighRotInrR:float)
   n122 = Look At - Bone: dyn_jacketFront_mstr_r(ComponentPose=Local To Component(LocalPose=Use cached pose 'main_SavedPose'()), Alpha=@n121) {
    alphaBoolBlend {blendOption=Linear}
    alphaScaleBiasClamp {bInterpResult=true}
    boneToModify {boneName=dyn_jacketFront_mstr_r}
    lookAtLocation=<0 -12 5>
    lookAtTarget {boneReference {boneName=calf_r}}
    lookAt_Axis {axis=<1 0>}
    lookUp_Axis {axis {x=0}}
   }
   n123 = Look At - Bone: dyn_jacketSide_mstr_r(ComponentPose=@n122, Alpha=@n121) {
    alphaBoolBlend {blendOption=Linear}
    alphaScaleBiasClamp {bInterpResult=true}
    boneToModify {boneName=dyn_jacketSide_mstr_r}
    lookAtLocation {x=0, z=11}
    lookAtTarget {boneReference {boneName=calf_r}}
    lookAt_Axis {axis=<1 0>}
    lookUp_Axis {axis {x=0}}
   }
   n124 = Get ThighRotUpR() -> (ThighRotUpR:float)
   n125 = "Transform (Modify) Bone - Bone: dyn_jacketFront_01_r"(ComponentPose=@n123, Alpha=@n124, Translation="0,0,0", Rotation="9.096,4.987,10.829", Scale="1,1,1") {
    rotationMode=BMM_Additive
    rotationSpace=BCS_BoneSpace
    translationMode=BMM_Additive
    translationSpace=BCS_BoneSpace
    alphaBoolBlend {blendOption=Linear}
    boneToModify {boneName=dyn_jacketFront_01_r}
    rotation=<rot 9.096 4.987 10.829>
   }
   n126 = "Transform (Modify) Bone - Bone: dyn_jacketSide_01_r"(ComponentPose=@n125, Alpha=@n124, Translation="0,0,0", Rotation="5.000,0.000,0.000", Scale="1,1,1") {
    rotationMode=BMM_Additive
    rotationSpace=BCS_BoneSpace
    translationMode=BMM_Additive
    translationSpace=BCS_BoneSpace
    alphaBoolBlend {blendOption=Linear}
    boneToModify {boneName=dyn_jacketSide_01_r}
    rotation {pitch=5}
   }
   n127 = Get ThighRotInrL() -> (ThighRotInrL:float)
   n128 = Look At - Bone: dyn_jacketFront_mstr_l(ComponentPose=@n126, Alpha=@n127) {
    alphaBoolBlend {blendOption=Linear}
    alphaScaleBiasClamp {bInterpResult=true}
    boneToModify {boneName=dyn_jacketFront_mstr_l}
    lookAtLocation=<0 12 -5>
    lookAtTarget {boneReference {boneName=calf_l}}
    lookAt_Axis {axis=<1 0>}
    lookUp_Axis {axis {x=0}}
   }
   n129 = Look At - Bone: dyn_jacketSide_mstr_l(ComponentPose=@n128, Alpha=@n127) {
    alphaBoolBlend {blendOption=Linear}
    alphaScaleBiasClamp {bInterpResult=true}
    boneToModify {boneName=dyn_jacketSide_mstr_l}
    lookAtLocation {x=0, z=-11}
    lookAtTarget {boneReference {boneName=calf_l}}
    lookAt_Axis {axis=<1 0>}
    lookUp_Axis {axis {x=0}}
   }
   n130 = Get ThighRotUpL() -> (ThighRotUpL:float)
   n131 = "Transform (Modify) Bone - Bone: dyn_jacketFront_01_l"(ComponentPose=@n129, Alpha=@n130, Translation="0,0,0", Rotation="9.096,-4.987,-10.828", Scale="1,1,1") {
    rotationMode=BMM_Additive
    rotationSpace=BCS_BoneSpace
    translationMode=BMM_Additive
    translationSpace=BCS_BoneSpace
    alphaBoolBlend {blendOption=Linear}
    boneToModify {boneName=dyn_jacketFront_01_l}
    rotation=<rot 9.096 -4.987 -10.8284>
   }
   n132 = "Transform (Modify) Bone - Bone: dyn_jacketSide_01_l"(ComponentPose=@n131, Alpha=@n130, Translation="0,0,0", Rotation="5,0,0", Scale="1,1,1") {
    rotationMode=BMM_Additive
    rotationSpace=BCS_BoneSpace
    translationMode=BMM_Additive
    translationSpace=BCS_BoneSpace
    alphaBoolBlend {blendOption=Linear}
    boneToModify {boneName=dyn_jacketSide_01_l}
    rotation {pitch=5}
   }
   n133 = Get JumpingAlpha() -> (JumpingAlpha:float)
   n134 = "Transform (Modify) Bone - Bone: dyn_jacketBack_01_l"(ComponentPose=@n132, Alpha=@n133, Translation="0,0,0", Rotation="35,10,0.000", Scale="1,1,1") {
    rotationMode=BMM_Additive
    rotationSpace=BCS_BoneSpace
    alphaBoolBlend {blendOption=Linear}
    alphaScaleBiasClamp {bInterpResult=true, interpSpeedDecreasing=2, interpSpeedIncreasing=5}
    boneToModify {boneName=dyn_jacketBack_01_l}
    rotation {pitch=35, yaw=10}
   }
   n135 = "Transform (Modify) Bone - Bone: dyn_jacketBack_01_r"(ComponentPose=@n134, Alpha=@n133, Translation="0,0,0", Rotation="35,-10,0.000", Scale="1,1,1") {
    rotationMode=BMM_Additive
    rotationSpace=BCS_BoneSpace
    alphaBoolBlend {blendOption=Linear}
    alphaScaleBiasClamp {bInterpResult=true, interpSpeedDecreasing=2, interpSpeedIncreasing=5}
    boneToModify {boneName=dyn_jacketBack_01_r}
    rotation {pitch=35, yaw=-10}
   }
   n136 = "Transform (Modify) Bone - Bone: dyn_jacketSide_01_l"(ComponentPose=@n135, Alpha=@n133, Translation="0,0,0", Rotation="15,-15,0.000", Scale="1,1,1") {
    rotationMode=BMM_Additive
    rotationSpace=BCS_BoneSpace
    alphaBoolBlend {blendOption=Linear}
    alphaScaleBiasClamp {bInterpResult=true, interpSpeedDecreasing=2, interpSpeedIncreasing=5}
    boneToModify {boneName=dyn_jacketSide_01_l}
    rotation {pitch=15, yaw=-15}
   }
   n137 = "Transform (Modify) Bone - Bone: dyn_jacketSide_01_l"(ComponentPose=@n136, Alpha=@n133, Translation="0,0,0", Rotation="15,15,0.000", Scale="1,1,1") {
    rotationMode=BMM_Additive
    rotationSpace=BCS_BoneSpace
    alphaBoolBlend {blendOption=Linear}
    alphaScaleBiasClamp {bInterpResult=true, interpSpeedDecreasing=2, interpSpeedIncreasing=5}
    boneToModify {boneName=dyn_jacketSide_01_l}
    rotation {pitch=15, yaw=15}
   }
   n138 = Get SkydiveAlpha() -> (SkydiveAlpha:float)
   n139 = "Transform (Modify) Bone - Bone: dyn_jacketBack_mstr_l"(ComponentPose=@n137, Alpha=@n138, Translation="0,0,0", Rotation="15,0,0.000", Scale="1,1,1") {
    rotationMode=BMM_Additive
    rotationSpace=BCS_BoneSpace
    alphaBoolBlend {blendOption=Linear}
    alphaScaleBiasClamp {bInterpResult=true, interpSpeedDecreasing=2, interpSpeedIncreasing=5}
    boneToModify {boneName=dyn_jacketBack_mstr_l}
    rotation {pitch=15}
   }
   n140 = "Transform (Modify) Bone - Bone: dyn_jacketBack_03_l"(ComponentPose=@n139, Alpha=@n138, Translation="0,0,0", Rotation="20,0,0.000", Scale="1,1,1") {
    rotationMode=BMM_Additive
    rotationSpace=BCS_BoneSpace
    alphaBoolBlend {blendOption=Linear}
    alphaScaleBiasClamp {bInterpResult=true, interpSpeedDecreasing=2, interpSpeedIncreasing=5}
    boneToModify {boneName=dyn_jacketBack_03_l}
    rotation {pitch=20}
   }
   n141 = "Transform (Modify) Bone - Bone: dyn_jacketBack_mstr_r"(ComponentPose=@n140, Alpha=@n138, Translation="0,0,0", Rotation="15,0,0.000", Scale="1,1,1") {
    rotationMode=BMM_Additive
    rotationSpace=BCS_BoneSpace
    alphaBoolBlend {blendOption=Linear}
    alphaScaleBiasClamp {bInterpResult=true, interpSpeedDecreasing=2, interpSpeedIncreasing=5}
    boneToModify {boneName=dyn_jacketBack_mstr_r}
    rotation {pitch=15}
   }
   n142 = Get ParachuteAlpha() -> (ParachuteAlpha:float)
   n143 = "Transform (Modify) Bone - Bone: dyn_jacketBack_01_l"(ComponentPose=@n141, Alpha=@n142, Translation="0,0,0", Rotation="15,10,0.000", Scale="1,1,1") {
    rotationMode=BMM_Additive
    rotationSpace=BCS_BoneSpace
    alphaBoolBlend {blendOption=Linear}
    alphaScaleBiasClamp {bInterpResult=true, interpSpeedDecreasing=2, interpSpeedIncreasing=5}
    boneToModify {boneName=dyn_jacketBack_01_l}
    rotation {pitch=15, yaw=10}
   }
   n144 = "Transform (Modify) Bone - Bone: dyn_jacketBack_01_r"(ComponentPose=@n143, Alpha=@n142, Translation="0,0,0", Rotation="15,-10,0.000", Scale="1,1,1") {
    rotationMode=BMM_Additive
    rotationSpace=BCS_BoneSpace
    alphaBoolBlend {blendOption=Linear}
    alphaScaleBiasClamp {bInterpResult=true, interpSpeedDecreasing=2, interpSpeedIncreasing=5}
    boneToModify {boneName=dyn_jacketBack_01_r}
    rotation {pitch=15, yaw=-10}
   }
   n145 = "Transform (Modify) Bone - Bone: dyn_jacketSide_01_l"(ComponentPose=@n144, Alpha=@n142, Translation="0,0,0", Rotation="15,-15,0.000", Scale="1,1,1") {
    rotationMode=BMM_Additive
    rotationSpace=BCS_BoneSpace
    alphaBoolBlend {blendOption=Linear}
    alphaScaleBiasClamp {bInterpResult=true, interpSpeedDecreasing=2, interpSpeedIncreasing=5}
    boneToModify {boneName=dyn_jacketSide_01_l}
    rotation {pitch=15, yaw=-15}
   }
   n146 = "Transform (Modify) Bone - Bone: dyn_jacketSide_01_l"(ComponentPose=@n145, Alpha=@n142, Translation="0,0,0", Rotation="15,15,0.000", Scale="1,1,1") {
    rotationMode=BMM_Additive
    rotationSpace=BCS_BoneSpace
    alphaBoolBlend {blendOption=Linear}
    alphaScaleBiasClamp {bInterpResult=true, interpSpeedDecreasing=2, interpSpeedIncreasing=5}
    boneToModify {boneName=dyn_jacketSide_01_l}
    rotation {pitch=15, yaw=15}
   }
   n147 = Component To Local(ComponentPose=@n146)
   n148 = "Transform (Modify) Bone - Bone: dyn_jacketFront_01_r"(
    ComponentPose=Local To Component(LocalPose=Use cached pose 'main_SavedPose'())
    Alpha=1
    Translation="0,0,0"
    Rotation="15,0,0"
    Scale="1,1,1"
   ) {
    rotationMode=BMM_Additive
    rotationSpace=BCS_BoneSpace
    translationMode=BMM_Additive
    translationSpace=BCS_BoneSpace
    alphaBoolBlend {blendOption=Linear}
    boneToModify {boneName=dyn_jacketFront_01_r}
    rotation {pitch=15}
   }
   n149 = "Transform (Modify) Bone - Bone: dyn_jacketSide_01_r"(ComponentPose=@n148, Alpha=1, Translation="0,0,0", Rotation="20,0,0", Scale="1,1,1") {
    rotationMode=BMM_Additive
    rotationSpace=BCS_BoneSpace
    translationMode=BMM_Additive
    translationSpace=BCS_BoneSpace
    alphaBoolBlend {blendOption=Linear}
    boneToModify {boneName=dyn_jacketSide_01_r}
    rotation {pitch=20}
   }
   n150 = "Transform (Modify) Bone - Bone: dyn_jacketBack_01_l"(ComponentPose=@n149, Alpha=@n29, Translation="0,0,0", Rotation="10,10,0.000", Scale="1,1,1") {
    rotationMode=BMM_Additive
    rotationSpace=BCS_BoneSpace
    alphaBoolBlend {blendOption=Linear}
    alphaScaleBiasClamp {bInterpResult=true, interpSpeedDecreasing=5}
    boneToModify {boneName=dyn_jacketBack_01_l}
    rotation {pitch=10, yaw=10}
   }
   n151 = "Transform (Modify) Bone - Bone: dyn_jacketBack_01_r"(ComponentPose=@n150, Alpha=@n29, Translation="0,0,0", Rotation="10,-10,0.000", Scale="1,1,1") {
    rotationMode=BMM_Additive
    rotationSpace=BCS_BoneSpace
    alphaBoolBlend {blendOption=Linear}
    alphaScaleBiasClamp {bInterpResult=true, interpSpeedDecreasing=5}
    boneToModify {boneName=dyn_jacketBack_01_r}
    rotation {pitch=10, yaw=-10}
   }
   n152 = "Transform (Modify) Bone - Bone: dyn_jacketBack_02_r"(ComponentPose=@n151, Alpha=@n29, Translation="0,0,0", Rotation="30,0,0.000", Scale="1,1,1") {
    rotationMode=BMM_Additive
    rotationSpace=BCS_BoneSpace
    alphaBoolBlend {blendOption=Linear}
    alphaScaleBiasClamp {bInterpResult=true, interpSpeedDecreasing=5}
    boneToModify {boneName=dyn_jacketBack_02_r}
    rotation {pitch=30}
   }
   n153 = "Transform (Modify) Bone - Bone: dyn_jacketFront_mstr_r"(ComponentPose=@n152, Alpha=Get ThighRotUpR(), Translation="0,0,3", Rotation="0,0,0", Scale="1,1,1") {
    rotationMode=BMM_Additive
    rotationSpace=BCS_BoneSpace
    translationMode=BMM_Additive
    translationSpace=BCS_BoneSpace
    alphaBoolBlend {blendOption=Linear}
    boneToModify {boneName=dyn_jacketFront_mstr_r}
    translation {z=3}
   }
   n154 = Get ThighRotOtrR() -> (ThighRotOtrR:float)
   n155 = "Transform (Modify) Bone - Bone: dyn_jacketFront_01_r"(ComponentPose=@n153, Alpha=@n154, Translation="0,0,0", Rotation="45,0,0", Scale="1,1,1") {
    rotationMode=BMM_Additive
    rotationSpace=BCS_BoneSpace
    translationMode=BMM_Additive
    translationSpace=BCS_BoneSpace
    alphaBoolBlend {blendOption=Linear}
    boneToModify {boneName=dyn_jacketFront_01_r}
    rotation {pitch=45}
   }
   n156 = "Transform (Modify) Bone - Bone: dyn_jacketFront_01_r"(ComponentPose=@n155, Alpha=@n154, Translation="0,0,2", Rotation="30,0,0", Scale="1,1,1") {
    rotationMode=BMM_Additive
    rotationSpace=BCS_BoneSpace
    translationMode=BMM_Additive
    translationSpace=BCS_BoneSpace
    alphaBoolBlend {blendOption=Linear}
    boneToModify {boneName=dyn_jacketFront_01_r}
    rotation {pitch=30}
    translation {z=2}
   }
   n157 = "Transform (Modify) Bone - Bone: dyn_jacketSide_02_r"(ComponentPose=@n156, Alpha=@n154, Translation="0,0,0", Rotation="45,0,0", Scale="1,1,1") {
    rotationMode=BMM_Additive
    rotationSpace=BCS_BoneSpace
    translationMode=BMM_Additive
    translationSpace=BCS_BoneSpace
    alphaBoolBlend {blendOption=Linear}
    boneToModify {boneName=dyn_jacketSide_02_r}
    rotation {pitch=45}
   }
   n158 = Get ThighRotUpL()
   n159 = "Transform (Modify) Bone - Bone: dyn_jacketFront_01_l"(ComponentPose=@n157, Alpha=@n158, Translation="0,0,0", Rotation="10,-15.000,-35", Scale="1,1,1") {
    rotationMode=BMM_Additive
    rotationSpace=BCS_BoneSpace
    translationMode=BMM_Additive
    translationSpace=BCS_BoneSpace
    alphaBoolBlend {blendOption=Linear}
    boneToModify {boneName=dyn_jacketFront_01_l}
    rotation=<rot 10 -15 -35>
   }
   n160 = "Transform (Modify) Bone - Bone: dyn_jacketFront_02_l"(ComponentPose=@n159, Alpha=@n158, Translation="0,0,0", Rotation="-25,0,0", Scale="1,1,1") {
    rotationMode=BMM_Additive
    rotationSpace=BCS_BoneSpace
    translationMode=BMM_Additive
    translationSpace=BCS_BoneSpace
    alphaBoolBlend {blendOption=Linear}
    boneToModify {boneName=dyn_jacketFront_02_l}
    rotation {pitch=-25}
   }
   n161 = "Transform (Modify) Bone - Bone: dyn_jacketSide_01_l"(ComponentPose=@n160, Alpha=@n158, Translation="0,-2,3", Rotation="0,-10,0", Scale="1,1,1") {
    rotationMode=BMM_Additive
    rotationSpace=BCS_BoneSpace
    translationMode=BMM_Additive
    translationSpace=BCS_BoneSpace
    alphaBoolBlend {blendOption=Linear}
    boneToModify {boneName=dyn_jacketSide_01_l}
    rotation {yaw=-10}
    translation {y=-2, z=3}
   }
   n162 = Get ThighRotOtrL() -> (ThighRotOtrL:float)
   n163 = "Transform (Modify) Bone - Bone: dyn_jacketFront_01_l"(ComponentPose=@n161, Alpha=@n162, Translation="0,0,0", Rotation="45,0,0", Scale="1,1,1") {
    rotationMode=BMM_Additive
    rotationSpace=BCS_BoneSpace
    translationMode=BMM_Additive
    translationSpace=BCS_BoneSpace
    alphaBoolBlend {blendOption=Linear}
    boneToModify {boneName=dyn_jacketFront_01_l}
    rotation {pitch=45}
   }
   n164 = "Transform (Modify) Bone - Bone: dyn_jacketFront_02_l"(ComponentPose=@n163, Alpha=@n162, Translation="0,0,2", Rotation="30,0,0", Scale="1,1,1") {
    rotationMode=BMM_Additive
    rotationSpace=BCS_BoneSpace
    translationMode=BMM_Additive
    translationSpace=BCS_BoneSpace
    alphaBoolBlend {blendOption=Linear}
    boneToModify {boneName=dyn_jacketFront_02_l}
    rotation {pitch=30}
    translation {z=2}
   }
   n165 = "Transform (Modify) Bone - Bone: dyn_jacketSide_02_l"(ComponentPose=@n164, Alpha=@n162, Translation="0,0,0", Rotation="45,0,0", Scale="1,1,1") {
    rotationMode=BMM_Additive
    rotationSpace=BCS_BoneSpace
    translationMode=BMM_Additive
    translationSpace=BCS_BoneSpace
    alphaBoolBlend {blendOption=Linear}
    boneToModify {boneName=dyn_jacketSide_02_l}
    rotation {pitch=45}
   }
   n166 = Component To Local(ComponentPose=@n165)
   n167 = Get bIsCrouching()
   n168 = Get BwdMovement() -> (BwdMovement:float)
   n169 = Trail controller - Bone: dyn_jacketBack_05_r!(
    ComponentPose=Local To Component(LocalPose=Blend Poses by bool(BlendPose_0=@n166, BlendPose_1=@n147, bActiveValue=@n167) {blendType=Linear})
    Alpha=@n168
    FakeVelocity:Vector="0,0,0"
   ) -> (Pose) {
    bActorSpaceFakeVel=true
    bShowTrailLocation=true
    chainLength=4
    debugLifeTime=0.1
    maxDeltaTime=0.0333
    relaxationSpeedScale=3
    alphaScaleBiasClamp {bInterpResult=true}
    baseJoint {boneName=root}
    planarLimits [
     {drivingBone {boneName=root}, planeTransform {location=<0 0 10>, rotation=<rot 0 0 0>, scale=<1 1 1>}}
     {drivingBone {boneName=spine_03}, planeTransform {location=<0 -20 0>, rotation=<rot 0 15 -89.9999>, scale=<1 1 1>}}
    ]
    rotationLimits [{limitMax=<180 180 180>}, {limitMax=<180 180 180>}, {limitMax=<180 180 180>}, {limitMax=<180 180 180>}]
    rotationOffsets [<0 0 0>, <0 0 0>, <0 0 0>, <0 0 0>]
    trailBone {boneName=dyn_jacketBack_05_r}
    trailRelaxationSpeed {editorCurveData {keys [{arriveTangent=-72.0739, interpMode=RCIM_Cubic, leaveTangent=-72.0739, tangentMode=RCTM_User, time=0, value=35}, <at 1 5 RCIM_Cubic>]}}
   }
   n170 = Trail controller - Bone: dyn_jacketBack_05_l!(ComponentPose=@n169, Alpha=@n168, FakeVelocity:Vector="0,0,0") -> (Pose) {
    bActorSpaceFakeVel=true
    bShowTrailLocation=true
    chainLength=4
    debugLifeTime=0.1
    maxDeltaTime=0.0333
    relaxationSpeedScale=3
    alphaScaleBiasClamp {bInterpResult=true}
    baseJoint {boneName=root}
    planarLimits [
     {drivingBone {boneName=root}, planeTransform {location=<0 0 10>, rotation=<rot 0 0 0>, scale=<1 1 1>}}
     {drivingBone {boneName=spine_03}, planeTransform {location=<0 -20 0>, rotation=<rot 0 15 -89.9999>, scale=<1 1 1>}}
    ]
    rotationLimits [{limitMax=<180 180 180>}, {limitMax=<180 180 180>}, {limitMax=<180 180 180>}, {limitMax=<180 180 180>}]
    rotationOffsets [<0 0 0>, <0 0 0>, <0 0 0>, <0 0 0>]
    trailBone {boneName=dyn_jacketBack_05_l}
    trailRelaxationSpeed {editorCurveData {keys [{arriveTangent=-72.0739, interpMode=RCIM_Cubic, leaveTangent=-72.0739, tangentMode=RCTM_User, time=0, value=35}, <at 1 5 RCIM_Cubic>]}}
   }
   n171 = Get LeftMovement() -> (LeftMovement:float)
   n172 = Trail controller - Bone: dyn_jacketFront_05_r!(ComponentPose=@n170, Alpha=@n171, FakeVelocity:Vector="0,0,0") -> (Pose) {
    bActorSpaceFakeVel=true
    bShowTrailLocation=true
    chainLength=4
    debugLifeTime=0.1
    maxDeltaTime=0.0333
    relaxationSpeedScale=3
    alphaScaleBiasClamp {bInterpResult=true}
    baseJoint {boneName=root}
    planarLimits [
     {drivingBone {boneName=root}, planeTransform {location=<0 0 10>, rotation=<rot 0 0 0>, scale=<1 1 1>}}
     {drivingBone {boneName=spine_03}, planeTransform {location=<0 -20 0>, rotation=<rot 0 15 -89.9999>, scale=<1 1 1>}}
    ]
    rotationLimits [{limitMax=<180 180 180>}, {limitMax=<180 180 180>}, {limitMax=<180 180 180>}, {limitMax=<180 180 180>}]
    rotationOffsets [<0 0 0>, <0 0 0>, <0 0 0>, <0 0 0>]
    trailBone {boneName=dyn_jacketFront_05_r}
    trailRelaxationSpeed {editorCurveData {keys [{arriveTangent=-72.0739, interpMode=RCIM_Cubic, leaveTangent=-72.0739, tangentMode=RCTM_User, time=0, value=35}, <at 1 5 RCIM_Cubic>]}}
   }
   n173 = Trail controller - Bone: dyn_jacketSide_05_r!(ComponentPose=@n172, Alpha=@n171, FakeVelocity:Vector="0,0,0") -> (Pose) {
    bActorSpaceFakeVel=true
    bShowTrailLocation=true
    chainLength=4
    debugLifeTime=0.1
    maxDeltaTime=0.0333
    relaxationSpeedScale=3
    alphaScaleBiasClamp {bInterpResult=true}
    baseJoint {boneName=root}
    planarLimits [
     {drivingBone {boneName=root}, planeTransform {location=<0 0 10>, rotation=<rot 0 0 0>, scale=<1 1 1>}}
     {drivingBone {boneName=spine_03}, planeTransform {location=<0 -20 0>, rotation=<rot 0 15 -89.9999>, scale=<1 1 1>}}
    ]
    rotationLimits [{limitMax=<180 180 180>}, {limitMax=<180 180 180>}, {limitMax=<180 180 180>}, {limitMax=<180 180 180>}]
    rotationOffsets [<0 0 0>, <0 0 0>, <0 0 0>, <0 0 0>]
    trailBone {boneName=dyn_jacketSide_05_r}
    trailRelaxationSpeed {editorCurveData {keys [{arriveTangent=-72.0739, interpMode=RCIM_Cubic, leaveTangent=-72.0739, tangentMode=RCTM_User, time=0, value=35}, <at 1 5 RCIM_Cubic>]}}
   }
   n174 = Get RightMovement() -> (RightMovement:float)
   n175 = Trail controller - Bone: dyn_jacketFront_05_l!(ComponentPose=@n173, Alpha=@n174, FakeVelocity:Vector="0,0,0") -> (Pose) {
    bActorSpaceFakeVel=true
    bShowTrailLocation=true
    chainLength=4
    debugLifeTime=0.1
    maxDeltaTime=0.0333
    relaxationSpeedScale=3
    alphaScaleBiasClamp {bInterpResult=true}
    baseJoint {boneName=root}
    planarLimits [
     {drivingBone {boneName=root}, planeTransform {location=<0 0 10>, rotation=<rot 0 0 0>, scale=<1 1 1>}}
     {drivingBone {boneName=spine_03}, planeTransform {location=<0 -20 0>, rotation=<rot 0 15 -89.9999>, scale=<1 1 1>}}
    ]
    rotationLimits [{limitMax=<180 180 180>}, {limitMax=<180 180 180>}, {limitMax=<180 180 180>}, {limitMax=<180 180 180>}]
    rotationOffsets [<0 0 0>, <0 0 0>, <0 0 0>, <0 0 0>]
    trailBone {boneName=dyn_jacketFront_05_l}
    trailRelaxationSpeed {editorCurveData {keys [{arriveTangent=-72.0739, interpMode=RCIM_Cubic, leaveTangent=-72.0739, tangentMode=RCTM_User, time=0, value=35}, <at 1 5 RCIM_Cubic>]}}
   }
   Trail controller - Bone: dyn_jacketSide_05_l!(ComponentPose=@n175, Alpha=@n174, FakeVelocity:Vector="0,0,0") -> (Pose) {
    bActorSpaceFakeVel=true
    bShowTrailLocation=true
    chainLength=4
    debugLifeTime=0.1
    maxDeltaTime=0.0333
    relaxationSpeedScale=3
    alphaScaleBiasClamp {bInterpResult=true}
    baseJoint {boneName=root}
    planarLimits [
     {drivingBone {boneName=root}, planeTransform {location=<0 0 10>, rotation=<rot 0 0 0>, scale=<1 1 1>}}
     {drivingBone {boneName=spine_03}, planeTransform {location=<0 -20 0>, rotation=<rot 0 15 -89.9999>, scale=<1 1 1>}}
    ]
    rotationLimits [{limitMax=<180 180 180>}, {limitMax=<180 180 180>}, {limitMax=<180 180 180>}, {limitMax=<180 180 180>}]
    rotationOffsets [<0 0 0>, <0 0 0>, <0 0 0>, <0 0 0>]
    trailBone {boneName=dyn_jacketSide_05_l}
    trailRelaxationSpeed {editorCurveData {keys [{arriveTangent=-72.0739, interpMode=RCIM_Cubic, leaveTangent=-72.0739, tangentMode=RCTM_User, time=0, value=35}, <at 1 5 RCIM_Cubic>]}}
   }
  ]}
 EventGraph {nodes [
   Event BlueprintUpdateAnimation()
   > n1 = Sequence!(execute) -> (then_0, then_1, then_2:Exec, then_3:Exec, then_4:Exec, then_5:Exec)
   > n2 = "Gravity Override State Function (Parameter Asset) (Nativized)"(
    CharacterPartAnimInstance=Self-Reference()
    GameTimeInSeconds=GetGameTimeInSeconds()
    Parameters=&/M_MED_SkeletonHunterGlitch_Parameters
   )
   n3 = "To Float (Boolean)"(InBool=Get bIsSkyDiving())
   > Set gravity_override(gravity_override=vector * vector(A=@n2, B=MapRangeClamped(Value=@n3, InRangeA=0, InRangeB=1, OutRangeA=1, OutRangeB=0.1)) -> (ReturnValue=~))
   n4 = Branch(execute=@n1.then_1, Condition=Get bIsOstrichVehicleDriver() -> (bIsOstrichVehicleDriver:bool))
   Set in vehicle rigid alpha(execute=[@n6.then, @n4.then, @n5.then], in vehicle rigid alpha:double=0) -> (Output_Get:double)
   > Set in vehicle scale alpha(in vehicle scale alpha:double=1) -> (Output_Get:double)
   n5 = Branch(execute=@n4.else, Condition=Get bIsFerretVehicleDriver() -> (bIsFerretVehicleDriver:bool))
   n6 = Branch(execute=@n5.else, Condition=Get bIsOctopusVehicleDriver() -> (bIsOctopusVehicleDriver:bool))
   Set in vehicle rigid alpha(execute=@n6.else, in vehicle rigid alpha=1)
   > Set in vehicle scale alpha(in vehicle scale alpha=0)
   n7 = PlayerMovement(execute=@n1.then_4, AnimInstance=Self-Reference(), XDotProductCoefficient=1, YDotProductCoefficient=1, ZDotProductCoefficient=1)
   n8 = MapRangeClamped(Value=@n7.YMovement, InRangeA=0, InRangeB=-1)
   n9 = float * float(A="Absolute (Float)"(A=@n7), B=0.5)
   > Set BwdMovement(BwdMovement:float=MapRangeClamped(
     Value=float + float(A=@n9, B=@n8) {is=title K2Node_CommutativeAssociativeBinaryOperator}
     InRangeA=0
     InRangeB=1.1
     OutRangeA=0.4
     OutRangeB=0
    )) -> (Output_Get:float)
   n10 = MapRangeClamped(Value=@n7, InRangeA=-0.2, InRangeB=1)
   n11 = float * float(A="Absolute (Float)"(A=@n7.YMovement), B=0.2)
   > Set RightMovement(RightMovement:float=MapRangeClamped(Value=float + float(A=@n11, B=@n10), InRangeA=0, InRangeB=1.2, OutRangeB=0.5)) -> (Output_Get:float)
   n12 = MapRangeClamped(Value=@n7, InRangeA=0.2, InRangeB=-1, OutRangeB=0.3)
   n13 = float * float(A="Absolute (Float)"(A=@n7.YMovement), B=0.2)
   > Set LeftMovement(LeftMovement:float=MapRangeClamped(Value=float + float(A=@n13, B=@n12), InRangeA=0, InRangeB=1.2, OutRangeB=0.5)) -> (Output_Get:float)
   n14 = GetDeltaTransformFromRefPose(self=GetOwnerSkeletalMeshComponent(), BoneName=thigh_l, BaseName=pelvis)
   n15 = MapRangeClamped(Value=BreakRotator(InRot=BreakTransform(InTransform=@n14).Rotation).Yaw, InRangeA=110, InRangeB=130)
   n16 = AND Boolean!(A=Get bIsCrouching(), B=Get bIsMoving2D(), C:bool=NOT Boolean(A=Get bIsSliding())) -> (ReturnValue)
   Set thigh_l_alpha(
    execute=@n1.then_5
    thigh_l_alpha:double=FInterpTo(Current=Get thigh_l_alpha() -> (thigh_l_alpha:double), Target=SelectFloat(A=@n15, B=0, bPickA=@n16), DeltaTime=0.4, InterpSpeed=0.9)
   ) -> (Output_Get:double)
   n17 = GetDeltaTransformFromRefPose(self=GetOwnerSkeletalMeshComponent(), BoneName=thigh_r, BaseName=pelvis)
   n18 = MapRangeClamped(Value=BreakRotator(InRot=BreakTransform(InTransform=@n17).Rotation).Pitch, InRangeA=5, InRangeB=30)
   n19 = AND Boolean(A=Get bIsCrouching(), B=Get bIsMoving2D(), C=NOT Boolean(A=Get bIsSliding()))
   > Set thigh_r_alpha(thigh_r_alpha:double=FInterpTo(Current=Get thigh_r_alpha(), Target=SelectFloat(A=@n18, B=0, bPickA=@n19), DeltaTime=0.4, InterpSpeed=0.9)) -> (Output_Get:double)
   n20 = Get bIsSliding()
   n21 = SelectFloat(
    A=float * float(A=Get crouch_moving() -> (crouch_moving:double), B=0.5) {is=title K2Node_PromotableOperator}
    B=0
    bPickA=AND Boolean(A=Get bIsCrouching(), B=Get bIsMoving2D(), C=NOT Boolean(A=@n20))
   )
   n22 = Get bIsCrouching()
   > Set crouch_moving_alpha(crouch_moving_alpha:double=FInterpTo(
     Current=Get crouch_moving_alpha() -> (crouch_moving_alpha:double)
     Target=SelectFloat(A=@n21, B=1, bPickA=@n22)
     DeltaTime=0.4
     InterpSpeed=1
    )) -> (Output_Get:double)
   n23 = Get bIsSliding()
   n24 = SelectFloat(A=float * float(A=Get crouch_moving(), B=0.15), B=0, bPickA=AND Boolean(A=Get bIsCrouching(), B=Get bIsMoving2D(), C=NOT Boolean(A=@n23)))
   n25 = Get bIsCrouching()
   > Set crouch_moving_lower_alpha(crouch_moving_lower_alpha:double=FInterpTo(
     Current=Get crouch_moving_lower_alpha() -> (crouch_moving_lower_alpha:double)
     Target=SelectFloat(A=@n24, B=1, bPickA=@n25)
     DeltaTime=0.4
     InterpSpeed=1
    )) -> (Output_Get:double)
   n26 = AND Boolean!(A=Get bIsCrouching(), B=NOT Boolean(A=Get bIsMoving2D())) -> (ReturnValue)
   n27 = Set Crouch Idle(
    execute=@n1.then_2
    Crouch Idle:double=FInterpTo(Current=Get Crouch Idle() -> (Crouch Idle:double), Target=SelectFloat(A=1, B=0, bPickA=@n26), DeltaTime=0.4, InterpSpeed=1)
   ) -> (Output_Get:double)
   n28 = GetDeltaTransformFromRefPose(self=GetOwnerSkeletalMeshComponent(), BoneName=pelvis, BaseName=root)
   n29 = MapRangeClamped(Value=BreakVector(InVec=BreakTransform(InTransform=@n28)), InRangeA=-50, InRangeB=-30, OutRangeA=1, OutRangeB=0)
   > Set Crouch Idle RBAN off(Crouch Idle RBAN off:double=SelectFloat(A=0, B=float * float(A=@n27, B=@n29), bPickA=Get bRigidBodyEnabled())) -> (Output_Get:double)
   n30 = Get bIsSliding()
   > Set Alpha_sliding(Alpha_sliding:double=FInterpTo(Current=Get Alpha_sliding(), Target="To Float (Boolean)"(InBool=@n30), DeltaTime=0.4, InterpSpeed=0.9)) -> (Output_Get:double)
   n31 = Get bRigidBodyEnabled()
   > Set RBAN off(RBAN off:double=SelectFloat(A=0, B=1, bPickA=@n31)) -> (Output_Get:double)
   > Set RBAN on(RBAN on:double=SelectFloat(A=1, B=0, bPickA=@n31)) -> (Output_Get:double)
   n32 = AND Boolean(A=Get bIsCrouching(), B=Get bIsMoving2D())
   > Set crouch_moving(crouch_moving:double=FInterpTo(Current=Get crouch_moving(), Target=SelectFloat(A=1, B=0, bPickA=@n32), DeltaTime=0.25, InterpSpeed=1)) -> (Output_Get:double)
   n33 = PlayerMovement(execute=@n1.then_3, AnimInstance=Self-Reference(), XDotProductCoefficient=1, YDotProductCoefficient=1, ZDotProductCoefficient=1)
   n34 = OR Boolean(A=Get bIsJumping(), B=Get bIsFalling())
   > Set Alpha_jump_right(Alpha_jump_right:double=float * float(A=MapRangeClamped(Value=@n33, InRangeA=0, InRangeB=1), B="To Float (Boolean)"(InBool=@n34))) -> (Output_Get:double)
  ]}
]
variables [
 Alpha_jump_right {type=number}
 Alpha_sliding {type=number}
 BwdMovement {type=number}
 Crouch Idle {type=number}
 Crouch Idle RBAN off {type=number}
 crouch_moving {type=number}
 crouch_moving_alpha {type=number}
 crouch_moving_lower_alpha {type=number}
 gravity_override {type=Vector}
 in vehicle rigid alpha {type=number}
 in vehicle scale alpha {type=number}
 LeftMovement {type=number}
 RBAN off {type=number}
 RBAN on {type=number}
 RightMovement {type=number}
 thigh_l_alpha {type=number}
 thigh_r_alpha {type=number}
]
